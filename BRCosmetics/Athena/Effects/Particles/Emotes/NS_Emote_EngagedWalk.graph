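# Niagara system for the EngagedWalk emote that pops up chat bubbles along with short screen-glow flashes.
# All three emitters burst one particle at start and another shortly after: a bubble, a cyan flare and a glow cone with mesh, light and sprite renderers.
# Bubbles overshoot in size, rise in steps from User.InitialOffset, then fade.
# A system-level random integer drives bubble sprite width and a green-or-white tint.
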NiagaraSystem /BRCosmetics/Athena/Effects/Particles/Emotes/NS_Emote_EngagedWalk
dependencies {
 usedDataInterfaces [
  &NiagaraDataInterfaceCurve
  &NiagaraDataInterfaceVectorCurve
  &NiagaraDataInterfaceColorCurve
  &NiagaraDataInterfaceVector2DCurve
 ]
 usedDynamicInputs [
  &MakeIntFromFloat
  &UniformAOrBFloat
  /Niagara/DynamicInputs/Random/RandomInt
 ]
 usedModules [
  &self:SystemSpawnScript.SystemScriptSource.SystemScriptGraph.NiagaraNodeAssignment_0.SetVariables_#1
  &SystemState
 ]
}
emitters {
 ChatBubble_Base001 {
  emitterUpdateScript [
   &EmitterState(Inactive Response=::NewEnumerator1, Life Cycle Mode=::NewEnumerator1, Loop Behavior=::NewEnumerator1, Loop Duration=1, UseLoopDelay=false)
   &SpawnRate(SpawnRate=20, Use Spawn Probability=false) {enabled=false}
   &SpawnBurst_Instantaneous(Spawn Count=1, Use Spawn Probability=false)
   &SpawnBurst_Instantaneous(Spawn Count=1, Spawn Time=0.45, Use Spawn Probability=false) {module=SpawnBurst_Instantaneous001}
  ]
  particleSpawnScript [
   /SetVariables(Particles.RandomInt:NiagaraInt32=$System.RandomInt) {isSetParameters=true}
   /Niagara/Modules/Spawn/Initialization/InitializeParticle(
    Color=<rgba 1 1 1 0.3333>
    Module.Write Color=true
    Module.Write Lifetime=true
    Module.Write Mass=true
    Module.Write Material Random=true
    Module.Write Position=true
    Module.Write Scale=false
    Module.Write SpriteRotation=false
    Module.Write SpriteSize=true
    Position=&SimulationPosition()
    Sprite Size=&MakeVector2D(
     X=&Multiply_Float(A=45, B=&MakeFloatFromInt(Int=$Particles.RandomInt))
     Y=45
    )
   )
   /Niagara/Modules/Spawn/Initialization/V2/InitializeParticle(Color=<rgba 1 1 1 0.333>, Sprite Size Mode=::NewEnumerator3, Uniform Sprite Size=50, UsePositionOffset=false, Write Material Random=true) {enabled=false, module=InitializeParticle001}
   /SetVariables(Particles.TomatoSubImageIndex:NiagaraFloat=&UniformRangedFloat(Minimum=0, Maximum=4)) {isSetParameters=true}
  ]
  particleUpdateScript [
   &ParticleState()
   &KillParticles(Kill Particles=/Game/Effects/Niagara/DynamicInputs/State/FN_IsExecutionStateActive()) {enabled=false}
   &ScaleColor(
    Scale Alpha=&FloatFromCurve(FloatCurve=DataInterface() {propertyValues {Curve {keys [
         <at 0 0 RCIM_Cubic>
         {arriveTangent=1.1111, leaveTangent=1.1111, time=0.1727, value=1}
         {arriveTangent=-9.7133, interpMode=RCIM_Cubic, leaveTangent=-9.7133, tangentMode=RCTM_User, time=0.9, value=1}
         {arriveTangent=-10.5682, interpMode=RCIM_Cubic, leaveTangent=-10.5683, tangentMode=RCTM_User, time=1, value=0}
        ]}}})
    Scale RGB=&MakeVectorFromLinearColorRGB(Color=<rgba 1 1 1 1>)
    ScaleA=true
    ScaleRGB=true
   )
   /SetVariables(
    Particles.TomatoColorAlpha:LinearColor=&MakeLinearColor(
     R:NiagaraFloat=1
     G:NiagaraFloat=1
     B:NiagaraFloat=1
     A:NiagaraFloat=&FloatFromCurve(FloatCurve=DataInterface() {propertyValues {Curve {keys [
          <at 0 0 RCIM_Cubic>
          {arriveTangent=1.1111, leaveTangent=1.1111, time=0.1727, value=1}
          {arriveTangent=-9.7133, interpMode=RCIM_Cubic, leaveTangent=-9.7133, tangentMode=RCTM_User, time=0.9, value=1}
          {arriveTangent=-10.5682, interpMode=RCIM_Cubic, leaveTangent=-10.5683, tangentMode=RCTM_User, time=1, value=0}
         ]}}})
    )
    Particles.TomatoSpriteSize:Vector2f=&Vector2DFromCurve(
     Vector2Curve=DataInterface() {propertyValues {
       XCurve {keys [
         {arriveTangent=-0.4757, interpMode=RCIM_Cubic, leaveTangent=-0.4757, tangentMode=RCTM_User, time=0, value=0}
         {arriveTangent=9.5393, interpMode=RCIM_Cubic, leaveTangent=9.5393, time=0.0543, value=1.2298}
         {arriveTangent=0.0567, interpMode=RCIM_Cubic, leaveTangent=0.0567, tangentMode=RCTM_User, time=0.1048, value=1}
         {interpMode=RCIM_Cubic, leaveTangent=6.1531, tangentMode=RCTM_Break, time=0.825, value=1.0003}
         {arriveTangent=-2.86, interpMode=RCIM_Cubic, leaveTangent=-2.86, tangentMode=RCTM_User, time=0.9, value=1}
         <at 1 0>
        ]}
       YCurve {keys [
         {arriveTangent=25.5866, interpMode=RCIM_Cubic, leaveTangent=25.5866, tangentMode=RCTM_User, time=0, value=0}
         {arriveTangent=7.1545, interpMode=RCIM_Cubic, leaveTangent=7.1545, time=0.0543, value=1.0976}
         {arriveTangent=0.0605, interpMode=RCIM_Cubic, leaveTangent=0.0605, tangentMode=RCTM_User, time=0.1048, value=0.75}
         {interpMode=RCIM_Cubic, leaveTangent=7.208, tangentMode=RCTM_Break, time=0.825, value=0.7503}
         {arriveTangent=-2.8427, interpMode=RCIM_Cubic, leaveTangent=-2.8427, tangentMode=RCTM_User, time=0.9, value=0.75}
         <at 1 0>
        ]}
      }}
     ScaleCurve=<35 35>
    )
   ) {isSetParameters=true}
   &ScaleSpriteSize(Scale Factor=&Vector2DFromCurve(
     Vector2Curve=DataInterface() {propertyValues {
       XCurve {keys [
         {arriveTangent=-0.4757, interpMode=RCIM_Cubic, leaveTangent=-0.4757, tangentMode=RCTM_User, time=0, value=0}
         {arriveTangent=9.5393, interpMode=RCIM_Cubic, leaveTangent=9.5393, time=0.0543, value=1.2298}
         {arriveTangent=0.0567, interpMode=RCIM_Cubic, leaveTangent=0.0567, tangentMode=RCTM_User, time=0.1048, value=1}
         {interpMode=RCIM_Cubic, leaveTangent=6.1531, tangentMode=RCTM_Break, time=0.825, value=1.0003}
         {arriveTangent=-2.86, interpMode=RCIM_Cubic, leaveTangent=-2.86, tangentMode=RCTM_User, time=0.9, value=1}
         <at 1 0>
        ]}
       YCurve {keys [
         {arriveTangent=25.5866, interpMode=RCIM_Cubic, leaveTangent=25.5866, tangentMode=RCTM_User, time=0, value=0}
         {arriveTangent=7.1545, interpMode=RCIM_Cubic, leaveTangent=7.1545, time=0.0543, value=1.0976}
         {arriveTangent=0.0605, interpMode=RCIM_Cubic, leaveTangent=0.0605, tangentMode=RCTM_User, time=0.1048, value=0.75}
         {interpMode=RCIM_Cubic, leaveTangent=7.208, tangentMode=RCTM_Break, time=0.825, value=0.7503}
         {arriveTangent=-2.8427, interpMode=RCIM_Cubic, leaveTangent=-2.8427, tangentMode=RCTM_User, time=0.9, value=0.75}
         <at 1 0>
        ]}
      }}
     ScaleCurve=<1 1>
    ))
   /SetVariables(
    Particles.Color=&MultiplyLinearColor(
     Linear Color=$Particles.Color
     Scale Factor=&MakeVector4_FromLinearColor(LinearColor=&LerpLinearColors(
       StartColor=<rgba 0.7015 1 0.2235 1>
       EndColor=<rgba 1 1 1 1>
       LerpFactor=&ClampFloat(Float=&MakeFloatFromInt(Int=$Particles.RandomInt))
      ))
    )
    Particles.Position=&Lerp_Vector(
     A=&Add_Vector(A=/Niagara/DynamicInputs/Helpers/SimulationPosition(), B=$User.InitialOffset)
     B=<0 0 300>
     Alpha=&FloatFromCurve(
      FloatCurve=DataInterface() {propertyValues {Curve {keys [<at 0 0>, <at 0.1 0.15>, <at 0.5524 0.15>, {arriveTangent=14.4019, leaveTangent=14.4019, time=0.6092, value=0.3}, <at 0.9 0.3>, <at 1 0.5>]}}}
      Scale Curve=0.85
     )
    )
   ) {isSetParameters=true}
  ]
  rendererClasses [&NiagaraSpriteRendererProperties]
  renderers [
   {rendererClass=&NiagaraSpriteRendererProperties}
   {rendererClass=&NiagaraSpriteRendererProperties, rendererIndex=1}
  ]
 }
 Flare {
  emitterUpdateScript [
   &EmitterScalability(
    Module.bApplySpawnCountScale=false
    Module.bApplySpawnCountScaleByDistanceFraction=false
    Module.bUseMaxDistance=false
    Module.bUseMinDistance=false
   ) {enabled=false}
   &EmitterLifeCycle(MaxLoopCount=1, NextLoopDuration=0.1) {enabled=false}
   &EmitterState()
   &SpawnBurst_Instantaneous(Spawn Count=1, Use Spawn Probability=false)
   &SpawnBurst_Instantaneous(Spawn Count=1, Spawn Time=0.45, Use Spawn Probability=false) {module=SpawnBurst_Instantaneous001}
  ]
  particleSpawnScript [
   /Niagara/Modules/Spawn/Initialization/InitializeParticle(
    Color=<rgba 0 1.7142 2 0.1>
    Lifetime=&UniformRangedFloat(Minimum=0.25, Maximum=0.35)
    Module.Write Color=true
    Module.Write Lifetime=true
    Module.Write Mass=true
    Module.Write Material Random=true
    Module.Write Position=true
    Module.Write Scale=false
    Module.Write SpriteRotation=true
    Module.Write SpriteSize=true
    Position=/Niagara/Functions/Localspace/SimulationPosition()
    Sprite Rotation=&UniformRangedFloat(Minimum=0, Maximum=360)
    Sprite Size=&Vector2DFromFloat(Value=&UniformRangedFloat(Minimum=100, Maximum=150))
   )
   &SystemLocation(MaskSpawnGroup=false, Offset=<0 0 5>)
  ]
  particleUpdateScript [
   &ParticleState()
   &SolveForcesAndVelocity(Clamp Velocity=false, Limit Acceleration=false)
   &ScaleColor(
    Scale Alpha=&FloatFromCurve(FloatCurve=DataInterface() {propertyValues {Curve {keys [<at -0.0046 1.0113>,<at 1.0011 0.002>]}}})
    Scale RGB=&VectorFromCurve(VectorCurve=DataInterface() {propertyValues {
       XCurve {keys [<at -0.0011 1.0067>,<at 0.9988 0>]}
       YCurve {keys [<at -0.0011 1.0067>,<at 0.9988 0>]}
       ZCurve {keys [<at -0.0011 1.0067>,<at 0.9988 0>]}
      }})
    ScaleA=true
    ScaleRGB=true
   )
  ]
  rendererClasses [&NiagaraSpriteRendererProperties]
  renderers [{rendererClass=&NiagaraSpriteRendererProperties}]
 }
 Local_Screen_Glow_Cone {
  emitterUpdateScript [
   &EmitterScalability(
    Module.bApplySpawnCountScale=false
    Module.bApplySpawnCountScaleByDistanceFraction=false
    Module.bUseMaxDistance=false
    Module.bUseMinDistance=false
   ) {enabled=false}
   &EmitterLifeCycle(MaxLoopCount=1, NextLoopDuration=0.1) {enabled=false}
   &EmitterState()
   &SpawnBurst_Instantaneous(Spawn Count=1, Use Spawn Probability=false)
   &SpawnBurst_Instantaneous(Spawn Count=1, Spawn Time=0.45, Use Spawn Probability=false) {module=SpawnBurst_Instantaneous001}
  ]
  particleSpawnScript [
   &InitializeParticle(
    Color=<rgba 2 2 2 0.3>
    Lifetime=&UniformRangedFloat(Minimum=0.25, Maximum=0.25)
    Module.Write Color=true
    Module.Write Lifetime=true
    Module.Write Mass=true
    Module.Write Material Random=true
    Module.Write Position=true
    Module.Write Scale=false
    Module.Write SpriteRotation=true
    Module.Write SpriteSize=true
    Position=&SimulationPosition()
    Sprite Rotation=&UniformRangedFloat(Minimum=0, Maximum=0)
    Sprite Size=&Vector2DFromFloat(Value=&UniformRangedFloat(Minimum=1, Maximum=20))
   )
   &AddVelocity(Velocity=<0 0 -0.01>)
   /SetVariables(Particles.LightPosition:Vector3f=/Niagara/DynamicInputs/Helpers/SimulationPosition()) {isSetParameters=true}
   /Niagara/Modules/Spawn/Orientation/V2/InitialMeshOrientation(Orientation Coordinate Space:ENiagaraCoordinateSpace=::Local, Rotation=<0.5 0 0>, Use Orientation Vector=true, Use Rotation Vector=true)
  ]
  particleUpdateScript [
   &ParticleState()
   &SolveForcesAndVelocity(Clamp Velocity=false, Limit Acceleration=false)
   /SetVariables(Particles.Scale=&VectorFromCurve(VectorCurve=DataInterface() {propertyValues {
       XCurve {keys [
         {arriveTangent=0.3216, interpMode=RCIM_Cubic, leaveTangent=0.3216, tangentMode=RCTM_User, time=0, value=0}
         {interpMode=RCIM_Cubic, tangentMode=RCTM_User, time=1, value=0.125}
        ]}
       YCurve {keys [
         {arriveTangent=0.2996, interpMode=RCIM_Cubic, leaveTangent=0.2996, tangentMode=RCTM_User, time=0, value=0}
         {interpMode=RCIM_Cubic, tangentMode=RCTM_User, time=1.0004, value=0.764}
        ]}
       ZCurve {keys [
         {arriveTangent=0.9164, interpMode=RCIM_Cubic, leaveTangent=0.9164, tangentMode=RCTM_User, time=0, value=0}
         {interpMode=RCIM_Cubic, tangentMode=RCTM_User, time=1, value=0.764}
        ]}
      }})) {isSetParameters=true}
   /SetVariables(
    Particles.LightColor=&ScaleLinearColorByCurve(
     InitialValue=<rgba 1 1 1 1>
     ScaleCurve=DataInterface() {propertyValues {
       AlphaCurve {keys [
         {interpMode=RCIM_Cubic, tangentMode=RCTM_User, time=0.0012, value=1}
         {arriveTangent=-2.3333, interpMode=RCIM_Cubic, leaveTangent=-2.3333, tangentMode=RCTM_User, time=0.1765, value=0.601}
         {arriveTangent=-1, interpMode=RCIM_Cubic, leaveTangent=-1, tangentMode=RCTM_User, time=0.2889, value=0.3}
         {interpMode=RCIM_Cubic, tangentMode=RCTM_User, time=1, value=0}
        ]}
       BlueCurve {keys [
         {interpMode=RCIM_Cubic, tangentMode=RCTM_User, time=0, value=0}
         {arriveTangent=2, interpMode=RCIM_Cubic, leaveTangent=2, tangentMode=RCTM_User, time=0.1765, value=0.2858}
         {interpMode=RCIM_Cubic, tangentMode=RCTM_User, time=0.503, value=11.4694}
         {interpMode=RCIM_Cubic, tangentMode=RCTM_User, time=1, value=0}
        ]}
       GreenCurve {keys [
         {interpMode=RCIM_Cubic, tangentMode=RCTM_User, time=0, value=0}
         {arriveTangent=1, interpMode=RCIM_Cubic, leaveTangent=1, tangentMode=RCTM_User, time=0.1765, value=0.1429}
         {arriveTangent=1.1038, interpMode=RCIM_Cubic, leaveTangent=1.1038, tangentMode=RCTM_User, time=0.5, value=0.5}
         {interpMode=RCIM_Cubic, tangentMode=RCTM_User, time=0.5, value=0.5}
         {interpMode=RCIM_Cubic, tangentMode=RCTM_User, time=1, value=0}
        ]}
       RedCurve {keys [
         {interpMode=RCIM_Cubic, tangentMode=RCTM_User, time=0, value=0}
         {arriveTangent=30, interpMode=RCIM_Cubic, leaveTangent=30, tangentMode=RCTM_User, time=0.1781, value=1.2533}
         {interpMode=RCIM_Cubic, tangentMode=RCTM_User, time=0.487, value=2.5303}
         {interpMode=RCIM_Cubic, tangentMode=RCTM_User, time=1, value=0}
        ]}
      }}
     CurveIndex=$Particles.NormalizedAge
    )
    Particles.LightIntensity:Vector3f=&VectorFromFloat(Value=0)
    Particles.SpriteRotation:NiagaraFloat=&FloatFromCurve(FloatCurve=DataInterface() {propertyValues {Curve {keys [<at -0.0046 50>,<at 1.0011 0.002>]}}})
   ) {isSetParameters=true}
   &ScaleColor(
    Scale Alpha=&FloatFromCurve(FloatCurve=DataInterface() {propertyValues {Curve {keys [<at 0.2596 1.0113 RCIM_Cubic>,<at 1.0011 0.002>]}}})
    Scale RGB=&VectorFromCurve(VectorCurve=DataInterface() {propertyValues {
       XCurve {keys [{arriveTangent=-1.6489, interpMode=RCIM_Cubic, leaveTangent=-1.6489, tangentMode=RCTM_User, time=-0.0011, value=1.0067}, <at 0.9988 0>]}
       YCurve {keys [{arriveTangent=-1.8566, interpMode=RCIM_Cubic, leaveTangent=-1.8566, tangentMode=RCTM_User, time=-0.0011, value=1.0067}, <at 0.9988 0>]}
       ZCurve {keys [{arriveTangent=-1.9787, interpMode=RCIM_Cubic, leaveTangent=-1.9787, tangentMode=RCTM_User, time=-0.0011, value=1.0067}, <at 0.9988 0>]}
      }})
    ScaleA=true
    ScaleRGB=true
   )
  ]
  rendererClasses [
   &NiagaraMeshRendererProperties
   &NiagaraLightRendererProperties
   &NiagaraSpriteRendererProperties
  ]
  renderers [
   {rendererClass=&NiagaraMeshRendererProperties, rendererIndex=0}
   {rendererClass=&NiagaraLightRendererProperties, rendererIndex=1}
   {rendererClass=&NiagaraSpriteRendererProperties, rendererIndex=2}
  ]
 }
}
summary {
 systemName=NS_Emote_EngagedWalk
 emitters [
  {
   emitterName=Local_Screen_Glow_Cone
   rendererClasses [
    &NiagaraMeshRendererProperties
    &NiagaraLightRendererProperties
    &NiagaraSpriteRendererProperties
   ]
  }
  {emitterName=Flare, rendererClasses [&NiagaraSpriteRendererProperties]}
  {emitterName=ChatBubble_Base001, rendererClasses [&NiagaraSpriteRendererProperties]}
 ]
 userVariables [User.InitialOffset {type=Vector3f, description="", defaultValue {struct=Vector3f, value=<0 0 12>}}]
}
