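# Niagara system for the one-shot poof in the Flying Kite emote, built from two burst sprite emitters, Smoke and Sparkles.
# Smoke throws gray, drag-slowed puffs that grow, fade, step through subimages and animate dynamic material parameters.
# Sparkles scatters bright warm glints that curl-noise and upward gravity push around as they pop in and shrink away.
# Water, pontoon and boost user variables are declared, but no shown module reads them.
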NiagaraSystem /BRCosmetics/Athena/Effects/Particles/Emotes/NS_Emote_FlyingKite_Poof
dependencies {
 usedDataInterfaces [&NiagaraDataInterfaceCurve]
 usedDynamicInputs []
 usedModules [/Niagara/Modules/System/SystemLifeCycle, &SystemState]
}
emitters {
 Smoke {
  emitterUpdateScript [
   &EmitterState()
   &SpawnBurst_Instantaneous(Spawn Count=30, Spawn Time=0.05, Use Spawn Probability=false) {module=SpawnBurst_Instantaneous001}
  ]
  particleSpawnScript [
   /Niagara/Modules/Spawn/Initialization/InitializeParticle(
    Color=&UniformRangedLinearColor(Minimum=<rgba 0.4323 0.4323 0.4323 0.35>, Maximum=<rgba 0.5255 0.5255 0.5255 0.5>)
    Lifetime=&UniformRangedFloat(Minimum=0.25, Maximum=0.5)
    Module.Write Color=true
    Module.Write Lifetime=true
    Module.Write Mass=true
    Module.Write Material Random=true
    Module.Write Position=false
    Module.Write Scale=false
    Module.Write SpriteRotation=true
    Module.Write SpriteSize=true
    Position=&SimulationPosition()
    Sprite Rotation=&UniformRangedFloat(Minimum=0, Maximum=360)
    Sprite Size=&Vector2DFromFloat(Value=&UniformRangedFloat(Minimum=10, Maximum=30))
   )
   &AccelerationForce(Acceleration=&UniformRangedVector(Minimum=<0 0 0>, Maximum=<0 0 50>))
   &AddVelocity(Velocity=&UniformRangedVector(Minimum=<-200 -200 -200>, Maximum=<200 200 200>))
   &ApplyInitialForces()
  ]
  particleUpdateScript [
   &ParticleState()
   &Drag(Drag=3, Use Linear Drag=true, Use Rotational Drag=false)
   &SolveForcesAndVelocity(Clamp Velocity=false, Limit Acceleration=false)
   &ScaleSpriteSize(Scale Factor=&Vector2DFromFloat(Value=&FloatFromCurve(FloatCurve=DataInterface() {propertyValues {Curve {keys [
          {arriveTangent=1.491, interpMode=RCIM_Cubic, leaveTangent=1.491, tangentMode=RCTM_User, time=-0.0046, value=1.1969}
          <at 1.0011 3.4809 RCIM_Cubic>
         ]}}})))
   /SetVariables(Particles.SubImageIndex=&FloatFromCurve(FloatCurve=DataInterface() {propertyValues {Curve {keys [
         {arriveTangent=29.8623, interpMode=RCIM_Cubic, leaveTangent=29.8623, tangentMode=RCTM_User, time=-0.0046, value=5.0556}
         <at 1.0011 22.1689 RCIM_Cubic>
        ]}}})) {isSetParameters=true}
   &DynamicMaterialParameters(
    Index 0 Param 1=&FloatFromCurve(FloatCurve=DataInterface() {propertyValues {Curve {keys [<at -0.0046 1.0113 RCIM_Cubic>,<at 1.0011 0.002>]}}})
    Index 0 Param 4=&FloatFromCurve(FloatCurve=DataInterface() {propertyValues {Curve {keys [<at 0 0 RCIM_Cubic>,<at 1 0.5>]}}})
    Module.Param0WriteEnabled=true
    Module.Param1WriteEnabled=false
    Module.Param2WriteEnabled=false
    Module.Param3WriteEnabled=true
   ) {module=DynamicMatParams}
   &CameraOffset(Camera Offset Amount=50)
   &ScaleColor(
    Scale Alpha=&FloatFromCurve(FloatCurve=DataInterface() {propertyValues {Curve {keys [
         {arriveTangent=-3.1757, interpMode=RCIM_Cubic, leaveTangent=-3.1757, tangentMode=RCTM_User, time=0.5, value=1.0113}
         {arriveTangent=-1.5121, interpMode=RCIM_Cubic, leaveTangent=-1.5121, tangentMode=RCTM_User, time=1.0011, value=0.002}
        ]}}})
    ScaleA=true
    ScaleRGB=false
   )
  ]
  rendererClasses [&NiagaraSpriteRendererProperties]
  renderers [{rendererClass=&NiagaraSpriteRendererProperties}]
 }
 Sparkles {
  emitterUpdateScript [
   &EmitterScalability(
    Module.bApplySpawnCountScale=false
    Module.bApplySpawnCountScaleByDistanceFraction=false
    Module.bUseMaxDistance=false
    Module.bUseMinDistance=false
   ) {enabled=false}
   &EmitterLifeCycle(NextLoopDuration=1) {enabled=false}
   &EmitterState()
   &SpawnBurst_Instantaneous(Spawn Count=13, Spawn Time=0.05, Use Spawn Probability=false)
  ]
  particleSpawnScript [
   &InitializeParticle(
    Color=<rgba 5 3.7043 1.51 1>
    Lifetime=&UniformRangedFloat(Minimum=0.3, Maximum=1)
    Module.Write Color=true
    Module.Write Lifetime=true
    Module.Write Mass=true
    Module.Write Material Random=true
    Module.Write Position=false
    Module.Write Scale=false
    Module.Write SpriteRotation=false
    Module.Write SpriteSize=true
    Position=&SimulationPosition()
    Sprite Size=&Vector2DFromFloat(Value=&UniformRangedFloat(Minimum=10, Maximum=30))
   )
   &AddVelocity(Velocity=&UniformRangedVector(Minimum=<-300 -300 -300>, Maximum=<300 300 300>))
   &ApplyInitialForces()
  ]
  particleUpdateScript [
   &ParticleState()
   &GravityForce(Gravity=<0 0 200>)
   /Niagara/Modules/Update/Forces/CurlNoiseForce(Noise Frequency=0.02, Noise Strength=500, Pan Noise=false)
   &Drag(Drag=5, Use Linear Drag=true, Use Rotational Drag=false)
   &SolveForcesAndVelocity(Clamp Velocity=false, Limit Acceleration=false)
   &ScaleColor(
    Scale Alpha=&FloatFromCurve(FloatCurve=DataInterface() {propertyValues {Curve {keys [<at -0.0092 0.0144>,<at 0.0967 0.9767>,<at 1.0011 0.002>]}}})
    ScaleA=true
    ScaleRGB=true
   )
   &ScaleSpriteSize(Scale Factor=&Vector2DFromFloat(Value=&FloatFromCurve(FloatCurve=DataInterface() {propertyValues {Curve {keys [<at -0.0046 0>,<at 0.1746 0.9952>,<at 1.0011 0.002>]}}})))
  ]
  rendererClasses [&NiagaraSpriteRendererProperties]
  renderers [{rendererClass=&NiagaraSpriteRendererProperties}]
 }
}
summary {
 systemName=NS_Emote_FlyingKite_Poof
 emitters [
  {emitterName=Smoke, rendererClasses [&NiagaraSpriteRendererProperties]}
  {emitterName=Sparkles, rendererClasses [&NiagaraSpriteRendererProperties]}
 ]
 userVariables [
  User.bIsInWater {type=NiagaraBool}
  User.Boost {type=NiagaraFloat, defaultValue=0, description=""}
  User.EffectScale {type=NiagaraFloat, defaultValue=1, description=""}
  User.EffectSpawnScale {type=NiagaraFloat, defaultValue=10, description=""}
  User.ForwardSpeed {type=NiagaraFloat, defaultValue=0, description=""}
  User.FrontPontoonBig.Depth {type=NiagaraFloat, defaultValue=0, description=""}
  User.FrontPontoonBig.Impulse {type=Vector3f, description="", defaultValue {struct=Vector3f, value=<0 0 0>}}
  User.GlobalOffset {type=Vector3f, description="", defaultValue {struct=Vector3f, value=<0 0 0>}}
  User.InWater {type=NiagaraBool}
  User.isBoosting {type=NiagaraBool}
  User.MidPontoonBig.Depth {type=NiagaraFloat, defaultValue=0, description=""}
  User.MidPontoonBig.Impulse {type=Vector3f, description="", defaultValue {struct=Vector3f, value=<0 0 0>}}
  User.NewVariable {type=NiagaraFloat, defaultValue=0, description=""}
  User.RearPontoonBig.Depth {type=NiagaraFloat, defaultValue=0, description=""}
  User.RearPontoonBig.Impulse {type=Vector3f, description="", defaultValue {struct=Vector3f, value=<0 0 0>}}
  User.WaterBodyIndex {type=NiagaraFloat, defaultValue=0, description=""}
  User.WaterDepth {type=NiagaraFloat, defaultValue=0, description=""}
  User.WaterPlaneLocation {type=Vector3f, description="", defaultValue {struct=Vector3f, value=<0 0 0>}}
  User.WaterPlaneNormal {type=Vector3f, description="", defaultValue {struct=Vector3f, value=<0 0 0>}}
  User.WaveTime {type=NiagaraFloat, defaultValue=0, description=""}
 ]
}
