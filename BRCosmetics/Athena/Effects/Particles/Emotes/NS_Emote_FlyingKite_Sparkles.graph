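# Niagara system for the flying kite emote that leaves a trail of glowing warm-gold sprite sparkles.
# A single Sparkles emitter spawns by distance moved rather than time, so particles appear as the emitter moves.
# Short-lived sparks burst in random directions, then drift under upward gravity, curl noise and drag while curves fade and scale them in and out.
# The water, pontoon and boost user parameters are not referenced by the modules shown.
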NiagaraSystem /BRCosmetics/Athena/Effects/Particles/Emotes/NS_Emote_FlyingKite_Sparkles
dependencies {
 usedDataInterfaces [&NiagaraDataInterfaceCurve]
 usedDynamicInputs []
 usedModules [/Niagara/Modules/System/SystemLifeCycle, &SystemState]
}
emitters {Sparkles {
  emitterUpdateScript [
   &EmitterState()
   &SpawnRate(SpawnRate=0, Use Spawn Probability=false)
   &SpawnPerUnit(
    Module.Use Max Movement Threshold=true
    Module.Use Movement Tolerance=true
    Movement Tolerance=5
    Spawn Per Unit=100
    Use Spawn Probability=false
   )
  ]
  particleSpawnScript [
   /Niagara/Modules/Spawn/Initialization/InitializeParticle(
    Color=<rgba 5 3.7043 1.51 1>
    Lifetime=&UniformRangedFloat(Minimum=0.3, Maximum=1)
    Module.Write Color=true
    Module.Write Lifetime=true
    Module.Write Mass=true
    Module.Write Material Random=true
    Module.Write Position=false
    Module.Write Scale=false
    Module.Write SpriteRotation=false
    Module.Write SpriteSize=true
    Position=&SimulationPosition()
    Sprite Size=&Vector2DFromFloat(Value=&UniformRangedFloat(Minimum=10, Maximum=30))
   )
   &AddVelocity(Velocity=&UniformRangedVector(Minimum=<-300 -300 -300>, Maximum=<300 300 300>))
   &ApplyInitialForces()
  ]
  particleUpdateScript [
   &ParticleState()
   &GravityForce(Gravity=<0 0 200>)
   /Niagara/Modules/Update/Forces/CurlNoiseForce(Noise Frequency=0.02, Noise Strength=500, Pan Noise=false)
   &Drag(Drag=5, Use Linear Drag=true, Use Rotational Drag=false)
   &SolveForcesAndVelocity(Clamp Velocity=false, Limit Acceleration=false)
   &ScaleColor(
    Scale Alpha=&FloatFromCurve(FloatCurve=DataInterface() {propertyValues {Curve {keys [<at -0.0092 0.0144>,<at 0.0967 0.9767>,<at 1.0011 0.002>]}}})
    ScaleA=true
    ScaleRGB=true
   )
   &ScaleSpriteSize(Scale Factor=&Vector2DFromFloat(Value=&FloatFromCurve(FloatCurve=DataInterface() {propertyValues {Curve {keys [<at -0.0046 0>,<at 0.1746 0.9952>,<at 1.0011 0.002>]}}})))
  ]
  rendererClasses [&NiagaraSpriteRendererProperties]
  renderers [{rendererClass=&NiagaraSpriteRendererProperties}]
 }}
summary {
 systemName=NS_Emote_FlyingKite_Sparkles
 emitters [{emitterName=Sparkles, rendererClasses [&NiagaraSpriteRendererProperties]}]
 userVariables [
  User.bIsInWater {type=NiagaraBool}
  User.Boost {type=NiagaraFloat, defaultValue=0, description=""}
  User.EffectScale {type=NiagaraFloat, defaultValue=1, description=""}
  User.EffectSpawnScale {type=NiagaraFloat, defaultValue=10, description=""}
  User.ForwardSpeed {type=NiagaraFloat, defaultValue=0, description=""}
  User.FrontPontoonBig.Depth {type=NiagaraFloat, defaultValue=0, description=""}
  User.FrontPontoonBig.Impulse {type=Vector3f, description="", defaultValue {struct=Vector3f, value=<0 0 0>}}
  User.GlobalOffset {type=Vector3f, description="", defaultValue {struct=Vector3f, value=<0 0 0>}}
  User.InWater {type=NiagaraBool}
  User.isBoosting {type=NiagaraBool}
  User.MidPontoonBig.Depth {type=NiagaraFloat, defaultValue=0, description=""}
  User.MidPontoonBig.Impulse {type=Vector3f, description="", defaultValue {struct=Vector3f, value=<0 0 0>}}
  User.NewVariable {type=NiagaraFloat, defaultValue=0, description=""}
  User.RearPontoonBig.Depth {type=NiagaraFloat, defaultValue=0, description=""}
  User.RearPontoonBig.Impulse {type=Vector3f, description="", defaultValue {struct=Vector3f, value=<0 0 0>}}
  User.WaterBodyIndex {type=NiagaraFloat, defaultValue=0, description=""}
  User.WaterDepth {type=NiagaraFloat, defaultValue=0, description=""}
  User.WaterPlaneLocation {type=Vector3f, description="", defaultValue {struct=Vector3f, value=<0 0 0>}}
  User.WaterPlaneNormal {type=Vector3f, description="", defaultValue {struct=Vector3f, value=<0 0 0>}}
  User.WaveTime {type=NiagaraFloat, defaultValue=0, description=""}
 ]
}
