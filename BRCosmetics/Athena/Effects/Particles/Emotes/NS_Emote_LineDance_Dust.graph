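# Niagara system for the Line Dance emote's dust puffs, with an added splash when in water.
# SmokeLocal always bursts four tan sprites that grow, step through flipbook frames and fade, as its bool switch is fixed false rather than bound to water state.
# Water_splash_arms001 bursts five velocity-aligned meshes only when User.bIsInWater (default true) is set, fading them through material and alpha curves.
# The other water user parameters are unreferenced in the listed modules.
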NiagaraSystem /BRCosmetics/Athena/Effects/Particles/Emotes/NS_Emote_LineDance_Dust
dependencies {
 usedDataInterfaces [&NiagaraDataInterfaceCurve, &NiagaraDataInterfaceVectorCurve]
 usedDynamicInputs []
 usedModules [/Niagara/Modules/System/SystemLifeCycle, &SystemState]
}
emitters {
 SmokeLocal {
  emitterUpdateScript [
   &EmitterState()
   &SpawnBurst_Instantaneous(
    Spawn Count=&Multiply_Int(A=&MakeCustomIntFromBool(Bool=false, True INT=0, False INT=1), B=4)
    Use Spawn Probability=false
   ) {module=SpawnBurst_Instantaneous001}
  ]
  particleSpawnScript [
   /Niagara/Modules/Spawn/Initialization/InitializeParticle(
    Color=&UniformRangedLinearColor(Minimum=<rgba 0.478 0.3791 0.2199 0.09>, Maximum=<rgba 0.781 0.7271 0.5498 0.156>)
    Lifetime=&UniformRangedFloat(Minimum=0.3, Maximum=0.7)
    Module.Write Color=true
    Module.Write Lifetime=true
    Module.Write Mass=true
    Module.Write Material Random=true
    Module.Write Position=false
    Module.Write Scale=false
    Module.Write SpriteRotation=true
    Module.Write SpriteSize=true
    Position=&SimulationPosition()
    Sprite Rotation=&UniformRangedFloat(Minimum=0, Maximum=360)
    Sprite Size=&Vector2DFromFloat(Value=&UniformRangedFloat(Minimum=5, Maximum=9))
   )
   /Niagara/Modules/Spawn/Location/SphereLocation(MaskSpawnGroup=false, Module.Override Randomness=false, Non Uniform Scale=<0 1 1>, Sphere Radius=5, Surface Only=false)
   &ApplyInitialForces()
   &CameraOffset(Camera Offset Amount=5)
  ]
  particleUpdateScript [
   &ParticleState()
   &Drag(Drag=5, Use Linear Drag=true, Use Rotational Drag=false)
   &SolveForcesAndVelocity(Clamp Velocity=false, Limit Acceleration=false)
   &ScaleSpriteSize(Scale Factor=&Vector2DFromFloat(Value=&FloatFromCurve(FloatCurve=DataInterface() {propertyValues {Curve {keys [
          {arriveTangent=1.491, interpMode=RCIM_Cubic, leaveTangent=1.491, tangentMode=RCTM_User, time=-0.0046, value=1.1969}
          <at 1.0011 3.4809 RCIM_Cubic>
         ]}}})))
   /SetVariables(Particles.SubImageIndex=&FloatFromCurve(FloatCurve=DataInterface() {propertyValues {Curve {keys [
         {arriveTangent=29.8623, interpMode=RCIM_Cubic, leaveTangent=29.8623, tangentMode=RCTM_User, time=-0.0046, value=5.0556}
         <at 1.0011 22.1689 RCIM_Cubic>
        ]}}})) {isSetParameters=true}
   &DynamicMaterialParameters(
    Index 0 Param 1=&FloatFromCurve(FloatCurve=DataInterface() {propertyValues {Curve {keys [<at -0.0046 1.0113 RCIM_Cubic>,<at 1.0011 0.002>]}}})
    Index 0 Param 4=&FloatFromCurve(FloatCurve=DataInterface() {propertyValues {Curve {keys [<at 0 0 RCIM_Cubic>,<at 1 0.5>]}}})
    Module.Param0WriteEnabled=true
    Module.Param1WriteEnabled=false
    Module.Param2WriteEnabled=false
    Module.Param3WriteEnabled=true
   ) {module=DynamicMatParams}
   &ScaleColor(
    Scale Alpha=&FloatFromCurve(FloatCurve=DataInterface() {propertyValues {Curve {keys [<at 0.5 1.0113 RCIM_Cubic>,<at 1.0011 0.002>]}}})
    ScaleA=true
    ScaleRGB=false
   )
  ]
  rendererClasses [&NiagaraSpriteRendererProperties]
  renderers [{rendererClass=&NiagaraSpriteRendererProperties}]
 }
 Water_splash_arms001 {
  emitterUpdateScript [
   &EmitterState()
   &SpawnBurst_Instantaneous(
    Spawn Count=&Multiply_Int(A=&MakeCustomIntFromBool(Bool=$User.bIsInWater, True INT=1, False INT=0), B=5)
    Use Spawn Probability=false
   )
  ]
  particleSpawnScript [
   &InitializeParticle(
    Lifetime=&UniformRangedFloat(Minimum=1, Maximum=1.2)
    Mass=&UniformRangedFloat(Minimum=0.6, Maximum=1.2)
    Mesh Scale=&UniformRangedVector(Minimum=<0.05 0.1 0.1>, Maximum=<0.1 0.25 0.2>)
    Module.Write Color=true
    Module.Write Lifetime=true
    Module.Write Mass=false
    Module.Write Material Random=true
    Module.Write Position=false
    Module.Write Scale=true
    Module.Write SpriteRotation=true
    Module.Write SpriteSize=false
    Position=&SimulationPosition()
    Sprite Rotation=&NormalizedAngleToDegrees(NormalizedAngle=&UniformRangedFloat(Minimum=0, Maximum=1))
   )
   &InheritVelocity(Module.Limit Speed=true, Velocity Scale=<0 0 0>)
   &TorusLocation(
    Align Torus With Axis=false
    Large Radius=50
    MaskSpawnGroup=false
    Module.Override Randomness=false
    Torus Mode=::NewEnumerator2
    Torus Origin:NiagaraPosition=&Add_Vector(A=Linked() {linkedVariable {name=Particles.Position, type=NiagaraPosition}}, B=<0 0 50>)
   ) {enabled=false}
   &PointForce(
    Force Origin=&SimulationPosition()
    Force Strength=90
    Linear Force Vector:Vector3f=<0 0 1>
    Module.Use Falloff=false
    Module.UseLinearForce=true
   )
   &AccelerationForce(Acceleration=<0 0 -50>)
   &ApplyInitialForces()
   /Niagara/Modules/Spawn/Orientation/V2/InitialMeshOrientation(
    Orientation Axis:Vector3f=<-1 0 0>
    Orientation Vector=$Particles.Velocity
    Rotation=&UniformRangedVector(Minimum=<0 0 0>, Maximum=<0 0 1>)
    Use Orientation Vector=true
    Use Rotation Vector=true
   )
  ]
  particleUpdateScript [
   &ParticleState() {module=UpdateAge}
   &DynamicMaterialParameters(
    Index 0 Param 1=&FloatFromCurve(FloatCurve=DataInterface() {propertyValues {Curve {keys [
         {arriveTangent=3.2513, interpMode=RCIM_Cubic, leaveTangent=3.2513, tangentMode=RCTM_User, time=0, value=-1.75}
         {arriveTangent=0.4055, interpMode=RCIM_Cubic, leaveTangent=0.4055, tangentMode=RCTM_User, time=0.9957, value=-0.1}
        ]}}})
    Index 0 Param 2=&FloatFromCurve(FloatCurve=DataInterface() {propertyValues {Curve {keys [<at 0 1>,<at 1 0>]}}})
    Module.Param0WriteEnabled=true
    Module.Param1WriteEnabled=false
    Module.Param2WriteEnabled=false
    Module.Param3WriteEnabled=false
   ) {module=DynamicMaterialParameters001}
   &ScaleColor(
    Scale Alpha=&FloatFromCurve(FloatCurve=DataInterface() {propertyValues {Curve {keys [<at -0.0046 1.0113 RCIM_Cubic>,<at 1.0011 0.002 RCIM_Cubic>]}}})
    ScaleA=true
    ScaleRGB=false
   )
   &Drag(Drag=3, Use Linear Drag=true, Use Rotational Drag=false)
   &SolveForcesAndVelocity(Clamp Velocity=false, Limit Acceleration=false)
   &ScaleMeshSize(Scale Factor=&VectorFromCurve(VectorCurve=DataInterface() {propertyValues {
       XCurve {keys [<at -0.0011 1.0067>,<at 1 0.5>]}
       YCurve {keys [<at -0.0011 1.0067>,<at 1 0.5>]}
       ZCurve {keys [<at -0.0011 1.1211>,<at 1 0.5>]}
      }})) {enabled=false, module=MeshSizeScale}
  ]
  rendererClasses [&NiagaraMeshRendererProperties]
  renderers [{rendererClass=&NiagaraMeshRendererProperties, rendererIndex=0}]
 }
}
summary {
 systemName=NS_Emote_LineDance_Dust
 emitters [
  {emitterName=SmokeLocal, rendererClasses [&NiagaraSpriteRendererProperties]}
  {emitterName=Water_splash_arms001, rendererClasses [&NiagaraMeshRendererProperties]}
 ]
 userVariables [
  User.bIsInWater {type=NiagaraBool, defaultValue=true}
  User.WaterBodyIndex {type=NiagaraFloat, defaultValue=0, description=""}
  User.WaterDepth {type=NiagaraFloat, defaultValue=0, description=""}
  User.WaterPlaneLocation {type=Vector3f, description="", defaultValue {struct=Vector3f, value=<0 0 0>}}
  User.WaterPlaneNormal {type=Vector3f, description="", defaultValue {struct=Vector3f, value=<0 0 0>}}
  User.WaveAttenuationFactor {type=NiagaraFloat, defaultValue=0, description=""}
  User.WaveTime {type=NiagaraFloat, defaultValue=0, description=""}
 ]
}
