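# Niagara one-shot effect for the Superhero Backflip emote, with separate ground and water variants.
# User.bIsInWater gates burst counts: false spawns an expanding ground smoke sprite, radial smoke puffs and speed-stretched streaks; true spawns droplets, a foam mesh and splash-arm meshes.
# Water emitters snap to the water plane, with curve-driven dynamic material parameters animating foam and arms.
# Debug, sprite splash and upward splash emitters are disabled.
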NiagaraSystem /BRCosmetics/Athena/Effects/Particles/Emotes/NS_Emote_Superhero_Backflip
dependencies {
 usedDataInterfaces [&NiagaraDataInterfaceCurve, &NiagaraDataInterfaceVectorCurve, &NiagaraDataInterfaceColorCurve]
 usedDynamicInputs []
 usedModules [/Niagara/Modules/System/SystemLifeCycle, &SystemState]
}
emitters {
 FN_Debug {
  bEnabled=false
  emitterUpdateScript [
   &EmitterState()
   &SpawnBurst_Instantaneous(Spawn Count=1, Use Spawn Probability=false)
  ]
  particleSpawnScript [/Niagara/Modules/Spawn/Initialization/InitializeParticle(
    Lifetime=0.1
    Module.Write Color=true
    Module.Write Lifetime=true
    Module.Write Mass=true
    Module.Write Material Random=true
    Module.Write Position=false
    Module.Write Scale=false
    Module.Write SpriteRotation=false
    Module.Write SpriteSize=true
    Position=&SimulationPosition()
    Sprite Size=<100 100>
   )]
  particleUpdateScript [
   &ParticleState()
   /SetVariables(Particles.Debug:NiagaraFloat=&MakeFloatFromVector(Vector=Linked() {linkedVariable {name=User.WaterPlaneLocation, type=NiagaraPosition}}, Channel=::NewEnumerator2)) {isSetParameters=true}
   &DynamicMaterialParameters(
    Index 0 Param 1=$Particles.Debug
    Module.Param0WriteEnabled=true
    Module.Param1WriteEnabled=true
    Module.Param2WriteEnabled=true
    Module.Param3WriteEnabled=true
   )
   &SystemLocation(MaskSpawnGroup=false, Offset=<-50 150 200>)
  ]
  rendererClasses [&NiagaraSpriteRendererProperties]
  renderers [{rendererClass=&NiagaraSpriteRendererProperties}]
 }
 FN_Splash {
  bEnabled=false
  emitterUpdateScript [
   &EmitterState()
   &SpawnBurst_Instantaneous(
    Spawn Count=&Multiply_Int(A=&MakeCustomIntFromBool(Bool=$User.bIsInWater, True INT=1, False INT=0), B=5)
    Use Spawn Probability=false
   ) {module=SpawnBurst_Instantaneous001}
   &SpawnBurst_Instantaneous(Spawn Count=5, Use Spawn Probability=false)
  ]
  particleSpawnScript [
   &InitializeParticle(
    Color=&UniformRangedLinearColor(Minimum=<rgba 0 0 0 1>, Maximum=<rgba 5 5 5 1>)
    Lifetime=&UniformRangedFloat(Minimum=0.4, Maximum=1)
    Module.Write Color=true
    Module.Write Lifetime=true
    Module.Write Mass=true
    Module.Write Material Random=true
    Module.Write Position=true
    Module.Write Scale=false
    Module.Write SpriteRotation=true
    Module.Write SpriteSize=true
    Position=&SimulationPosition()
    Sprite Rotation=&UniformRangedFloat(Minimum=0, Maximum=360)
    Sprite Size=&Vector2DFromFloat(Value=&UniformRangedFloat(Minimum=30, Maximum=60))
   )
   /NiagaraFN/Sprite/FN_RandomizeSpriteUV()
   /Niagara/Modules/Update/SubUV/SubUVAnimation(SubUV Lookup Index=&UniformRangedFloat(Minimum=0, Maximum=1))
   &AddVelocity(Velocity=&UniformRangedVector(Minimum=<-300 -300 300>, Maximum=<300 300 300>))
   &Color(Color=&UniformRangedLinearColor(Minimum=<rgba 1 1 1 1>, Maximum=<rgba 2 2 2 2>))
  ]
  particleUpdateScript [
   &ParticleState()
   &GravityForce(Gravity=<0 0 -1600>)
   &SolveForcesAndVelocity(Clamp Velocity=false, Limit Acceleration=false)
   &ScaleSpriteSize(Scale Factor=&Multiply_Vector2(
     A=<2 2>
     B=&Vector2DFromFloat(Value=&Multiply_Float(
       A=1
       B=&FloatFromCurve(FloatCurve=DataInterface() {propertyValues {Curve {keys [
            {arriveTangent=2.0364, interpMode=RCIM_Cubic, leaveTangent=2.0364, tangentMode=RCTM_User, time=0.0023, value=0.3733}
            {arriveTangent=0.7301, interpMode=RCIM_Cubic, leaveTangent=0.7301, time=0.3082, value=0.6733}
            <at 0.3112 0.6755>
            {arriveTangent=0.617, interpMode=RCIM_Cubic, leaveTangent=0.617, tangentMode=RCTM_User, time=0.6894, value=0.875}
            <at 0.994 1.0109>
           ]}}})
      ))
    ))
   &ScaleColor(
    Scale Alpha=&FloatFromCurve(FloatCurve=DataInterface() {propertyValues {Curve {keys [
         {interpMode=RCIM_Cubic, tangentMode=RCTM_User, time=-0.0007, value=-0.0008}
         {arriveTangent=0.0027, interpMode=RCIM_Cubic, leaveTangent=0.0027, tangentMode=RCTM_User, time=0.0821, value=0.9831}
         {arriveTangent=-1.0675, interpMode=RCIM_Cubic, leaveTangent=-1.0675, tangentMode=RCTM_User, time=0.8646, value=0.7677}
         <at 1.0011 0.002>
        ]}}})
    Scale RGB=&ScaleFloatByCurve(
     InitialValue:Vector3f=<1 1 1>
     ScaleCurve:NiagaraDataInterfaceVectorCurve=DataInterface() {propertyValues {
       XCurve {keys [<at 0 0>,<at 0.3112 0.6225>,<at 0.5 1>,<at 1 0>]}
       YCurve {keys [<at 0 0>,<at 0.3112 0.6225>,<at 0.5 1>,<at 1 0>]}
       ZCurve {keys [<at 0 0>,<at 0.3112 0.6225>,<at 0.5 1>,<at 1 0>]}
      }}
     CurveIndex:NiagaraFloat=$Particles.NormalizedAge
    )
    ScaleA=true
    ScaleRGB=true
   )
   &DynamicMaterialParameters(
    Index 0 Param 1=&FloatFromCurve(FloatCurve=DataInterface() {propertyValues {Curve {keys [
         {interpMode=RCIM_Cubic, tangentMode=RCTM_User, time=-0.0046, value=1.0113}
         {arriveTangent=-1.0036, interpMode=RCIM_Cubic, leaveTangent=-1.0036, tangentMode=RCTM_User, time=0.705, value=0.8464}
         <at 1.0011 0.002>
        ]}}})
    Module.Param0WriteEnabled=true
    Module.Param1WriteEnabled=true
    Module.Param2WriteEnabled=true
    Module.Param3WriteEnabled=true
   )
  ]
  rendererClasses [&NiagaraSpriteRendererProperties]
  renderers [{rendererClass=&NiagaraSpriteRendererProperties}]
 }
 FN_Water_Drops {
  emitterUpdateScript [
   &EmitterState()
   &SpawnBurst_Instantaneous(
    Spawn Count=&Multiply_Int(A=&MakeCustomIntFromBool(Bool=$User.bIsInWater, True INT=1, False INT=0), B=50)
    Use Spawn Probability=false
   ) {module=SpawnBurst_Instantaneous001}
   &SpawnBurst_Instantaneous(Spawn Count=50, Use Spawn Probability=false) {enabled=false}
  ]
  particleSpawnScript [
   /SetVariables(Particles.Random=&UniformRangedFloat(Minimum=0, Maximum=1)) {isSetParameters=true}
   &InitializeParticle(
    Color=<rgba 5 5 5 1>
    Lifetime=&UniformRangedFloat(Minimum=0.3, Maximum=0.55)
    Mass=&UniformRangedFloat(Minimum=0.6, Maximum=1.2)
    Mesh Scale=<100 100 100>
    Module.Write Color=true
    Module.Write Lifetime=true
    Module.Write Mass=true
    Module.Write Material Random=true
    Module.Write Position=false
    Module.Write Scale=false
    Module.Write SpriteRotation=true
    Module.Write SpriteSize=true
    Position=&SimulationPosition()
    Sprite Size=&UniformRangedVector2D(Minimum=<2 4>, Maximum=<4 8>)
   )
   &DynamicMaterialParameters(
    Index 0 Param 1=&UniformRangedFloat(Minimum=4, Maximum=10)
    Module.Param0WriteEnabled=true
    Module.Param1WriteEnabled=false
    Module.Param2WriteEnabled=false
    Module.Param3WriteEnabled=false
   )
   &InitializeWaterSnapToOwner()
   &ConstrainPositionToPlane(Plane Normal:Vector3f=$Particles.WaterPlaneNormal, Plane Position:NiagaraPosition=$User.WaterPlaneLocation) {module=ConstrainPositionToPlane001}
   &CylinderLocation(
    Cylinder Height=1
    Cylinder Radius=20
    Lathe Profile=DataInterface() {propertyValues {Curve {keys []}}}
    MaskSpawnGroup=false
    Module.Enable Lathe Profile=false
    Module.Override Randomness=false
    Module.Surface Only=true
    Surface Only Band Thickness=2
    Use Endcaps In Surface Only Mode=false
   )
   /SetVariables(Particles.SpriteSize=&Multiply_Vector2(
     A=$Particles.SpriteSize
     B=&Vector2DFromFloat(Value=&FloatFromCurve(
       FloatCurve=DataInterface() {propertyValues {Curve {keys [
           {arriveTangent=1.6964, interpMode=RCIM_Cubic, time=-0.0012, value=-0.0071}
           {arriveTangent=13.7683, interpMode=RCIM_Cubic, leaveTangent=13.7683, time=0.3287, value=17.2263}
           {arriveTangent=-3.3438, interpMode=RCIM_Cubic, leaveTangent=-3.3438, time=0.7139, value=9.8392}
           {arriveTangent=1.8535, interpMode=RCIM_Cubic, leaveTangent=1.8535, tangentMode=RCTM_User, time=0.9945, value=15}
          ]}}}
       CurveIndex=$Emitter.Scalability.DistanceFraction
      ))
    )) {enabled=false, isSetParameters=true}
   &AddVelocityFromPoint(
    Module.Use Falloff Distance=false
    Velocity Origin:Vector3f=&SimulationPosition()
    Velocity Strength=&UniformRangedFloat(Minimum=350, Maximum=500)
   )
   &ApplyInitialForces()
  ]
  particleUpdateScript [
   &ParticleState()
   &ScaleColor(
    Scale Alpha=&FloatFromCurve(FloatCurve=DataInterface() {propertyValues {Curve {keys [
         <at 0 0 RCIM_Cubic>
         {arriveTangent=0.002, interpMode=RCIM_Cubic, leaveTangent=0.002, time=0.1183, value=1.0113}
         <at 1.0011 0.002 RCIM_Cubic>
        ]}}})
    ScaleA=true
    ScaleRGB=true
   )
   /Niagara/Modules/Update/Forces/CurlNoiseForce(
    Noise Frequency=1
    Noise Strength=&FloatFromCurve(
     FloatCurve=DataInterface() {propertyValues {Curve {keys [
         {arriveTangent=-1159.6794, interpMode=RCIM_Cubic, leaveTangent=-1159.6792, tangentMode=RCTM_User, time=-0.0046, value=500}
         {arriveTangent=-667.107, interpMode=RCIM_Cubic, leaveTangent=-667.11, tangentMode=RCTM_User, time=0.2323, value=224.675}
         <at 1.0011 50 RCIM_Cubic>
        ]}}}
     Scale Curve=5
    )
    Pan Noise=false
   )
   &GravityForce()
   &Drag(Use Linear Drag=true, Use Rotational Drag=false)
   &SolveForcesAndVelocity(Clamp Velocity=false, Limit Acceleration=false)
  ]
  rendererClasses [&NiagaraSpriteRendererProperties]
  renderers [{rendererClass=&NiagaraSpriteRendererProperties}]
 }
 FN_Water_Foam {
  emitterUpdateScript [
   &EmitterState()
   &EmitterScalability(
    MaxDistance=4000
    Module.bApplySpawnCountScale=true
    Module.bApplySpawnCountScaleByDistanceFraction=true
    Module.bUseMaxDistance=true
    Module.bUseMinDistance=false
    SpawnCountScaleByDistanceFraction=DataInterface() {propertyValues {Curve {keys [<at 0 1>,<at 1 0.5>]}}}
   ) {enabled=false}
   &SpawnPerUnit(Module.Use Max Movement Threshold=true, Module.Use Movement Tolerance=false, Spawn Per Unit=0, Use Spawn Probability=false) {enabled=false}
   &SpawnBurst_Instantaneous(
    Spawn Count=&Multiply_Int(A=&MakeCustomIntFromBool(Bool=$User.bIsInWater, True INT=1, False INT=0), B=1)
    Use Spawn Probability=false
   )
   &EmitterLifeCycle(MaxLoopCount=1, NextLoopDuration=1) {enabled=false}
   &SpawnRate(SpawnRate=0, Use Spawn Probability=false) {enabled=false}
   &SpawnBurst_Instantaneous(Spawn Count=1, Use Spawn Probability=false) {enabled=false, module=SpawnBurst_Instantaneous001}
  ]
  particleSpawnScript [
   &InitializeParticle(
    Lifetime=&UniformRangedFloat(Minimum=1, Maximum=3)
    Mass=&UniformRangedFloat(Minimum=0.6, Maximum=1.2)
    Mesh Scale=&VectorFromFloat(Value=70)
    Module.Write Color=true
    Module.Write Lifetime=true
    Module.Write Mass=true
    Module.Write Material Random=true
    Module.Write Position=false
    Module.Write Scale=true
    Module.Write SpriteRotation=true
    Module.Write SpriteSize=false
    Position=&SimulationPosition()
    Sprite Rotation=&NormalizedAngleToDegrees(NormalizedAngle=&UniformRangedFloat(Minimum=0, Maximum=1))
   ) {enabled=false}
   /Niagara/Modules/Spawn/Initialization/V2/InitializeParticle(
    Lifetime=&UniformRangedFloat(Minimum=2, Maximum=3)
    Mass=&UniformRangedFloat(Minimum=0.6, Maximum=1.2)
    Mass Mode=::NewEnumerator1
    Mesh Scale Mode=::NewEnumerator3
    Mesh Uniform Scale=100
    Sprite Rotation Angle=&UniformRangedFloat(Minimum=0, Maximum=1)
    Sprite Rotation Mode=::NewEnumerator1
    UsePositionOffset=false
    Write Material Random=true
   ) {module=InitializeParticle001}
   /Niagara/Modules/Spawn/Orientation/V2/InitialMeshOrientation(
    Orientation Vector=<0 0 0>
    Rotation=&UniformRangedVector(Minimum=<0 0 0>, Maximum=<0 0 1>)
    Use Orientation Vector=true
    Use Rotation Vector=true
   ) {enabled=false}
   /Niagara/Modules/Spawn/Orientation/V3/InitialMeshOrientation(
    Mesh Orientation Mode=::NewEnumerator2
    Orientation Vector=<0 0 0>
    Rotation=&UniformRangedVector(Minimum=<0 0 0>, Maximum=<0 0 1>)
    Use Rotation Vector=true
   ) {module=InitialMeshOrientation001}
   &DynamicMaterialParameters(
    Index 1 Param 3=115
    Index 3 Param 1=$User.WaterBodyIndex
    Index 3 Param 2=$User.WaterDepth
    Module.Param0WriteEnabled1=false
    Module.Param0WriteEnabled3=true
    Module.Param1WriteEnabled1=false
    Module.Param1WriteEnabled3=true
    Module.Param2WriteEnabled1=true
    Module.Param2WriteEnabled3=true
    Module.Param3WriteEnabled1=false
    Module.Param3WriteEnabled3=true
    Write Parameter Index 0=false
    Write Parameter Index 1=true
    Write Parameter Index 3=true
   )
   &InitializeWaterSnapToOwner()
   &InheritVelocity(Module.Limit Speed=true)
  ]
  particleUpdateScript [
   &ParticleState()
   &DynamicMaterialParameters(
    Index 0 Param 1=&FloatFromCurve(FloatCurve=DataInterface() {propertyValues {Curve {keys [<at -0.0046 1>,<at 1.0011 4>]}}})
    Index 0 Param 2=&FloatFromCurve(FloatCurve=DataInterface() {propertyValues {Curve {keys [{arriveTangent=0.6218, interpMode=RCIM_Cubic, leaveTangent=0.6218, tangentMode=RCTM_User, time=-0.0046, value=0.5}, <at 1.0011 1 RCIM_Cubic>]}}})
    Index 0 Param 3=&FloatFromCurve(FloatCurve=DataInterface() {propertyValues {Curve {keys [<at -0.0046 1>,<at 1.0011 0.5>]}}})
    Index 2 Param 1=&FloatFromCurve(FloatCurve=DataInterface() {propertyValues {Curve {keys [<at -0.0046 0.4>,<at 1.0011 1.5>]}}})
    Index 2 Param 2=&FloatFromCurve(FloatCurve=DataInterface() {propertyValues {Curve {keys [<at -0.0046 0.1>,<at 0.11 2.6>,<at 1.0011 0.3>]}}})
    Index 2 Param 3=1
    Index 2 Param 4=&FloatFromCurve(FloatCurve=DataInterface() {propertyValues {Curve {keys [
         {arriveTangent=-1.9991, interpMode=RCIM_Cubic, time=0, value=1}
         {arriveTangent=-2.7143, interpMode=RCIM_Cubic, leaveTangent=-2.7143, tangentMode=RCTM_User, time=0.2095, value=0.6901}
         <at 1 0 RCIM_Cubic>
        ]}}})
    Module.Param0WriteEnabled=true
    Module.Param0WriteEnabled2=true
    Module.Param1WriteEnabled=true
    Module.Param1WriteEnabled2=true
    Module.Param2WriteEnabled=true
    Module.Param2WriteEnabled2=true
    Module.Param3WriteEnabled=true
    Module.Param3WriteEnabled2=true
    Write Parameter Index 2=true
   ) {module=DynamicMaterialParameters001}
   &ScaleColor(
    Scale Alpha=&FloatFromCurve(FloatCurve=DataInterface() {propertyValues {Curve {keys [<at -0.0046 1.0113 RCIM_Cubic>,<at 1.0011 0.002 RCIM_Cubic>]}}})
    ScaleA=true
    ScaleRGB=true
   ) {enabled=false}
   &Drag(Drag=2, Use Linear Drag=true, Use Rotational Drag=false)
   &SolveForcesAndVelocity(Clamp Velocity=false, Limit Acceleration=false)
   &ScaleMeshSize(Scale Factor=&VectorFromFloat(Value=&FloatFromCurve(FloatCurve=DataInterface() {propertyValues {Curve {keys [
          {arriveTangent=26.4511, interpMode=RCIM_Cubic, leaveTangent=26.451, tangentMode=RCTM_User, time=-0.0046, value=0.6073}
          {arriveTangent=3.3733, interpMode=RCIM_Cubic, leaveTangent=3.3733, time=0.1268, value=2.5016}
          <at 1.0011 4 RCIM_Cubic>
         ]}}}))) {module=MeshSizeScale}
   &ConstrainPositionToPlane(
    Plane Normal:Vector3f=$Particles.WaterPlaneNormal
    Plane Position:NiagaraPosition=&Add_Vector(A=$User.WaterPlaneLocation, B=<0 0 0>)
   )
   &CameraOffset(Camera Offset Amount=10)
  ]
  rendererClasses [&NiagaraMeshRendererProperties]
  renderers [{rendererClass=&NiagaraMeshRendererProperties, rendererIndex=0}]
 }
 SmokeGround {
  emitterUpdateScript [
   &EmitterState()
   &SpawnBurst_Instantaneous(
    Spawn Count=&Multiply_Int(A=&MakeCustomIntFromBool(Bool=$User.bIsInWater, True INT=0, False INT=1), B=10)
    Use Spawn Probability=false
   ) {module=SpawnBurst_Instantaneous001}
  ]
  particleSpawnScript [
   /Niagara/Modules/Spawn/Initialization/InitializeParticle(
    Color=<rgba 1 1 1 0.55>
    Lifetime=0.3
    Module.Write Color=true
    Module.Write Lifetime=true
    Module.Write Mass=true
    Module.Write Material Random=true
    Module.Write Position=false
    Module.Write Scale=false
    Module.Write SpriteRotation=true
    Module.Write SpriteSize=true
    Position=&SimulationPosition()
    Sprite Size=&Vector2DFromFloat(Value=290)
   )
   /SetVariables(Particles.SpriteFacing=<0 0 1>) {isSetParameters=true}
   &SystemLocation(MaskSpawnGroup=false, Offset=<0 0 5>)
  ]
  particleUpdateScript [
   &ParticleState()
   &ScaleSpriteSize(Scale Factor=&Vector2DFromFloat(Value=&FloatFromCurve(FloatCurve=DataInterface() {propertyValues {Curve {keys [
          {arriveTangent=1.0656, interpMode=RCIM_Cubic, leaveTangent=1.0656, tangentMode=RCTM_User, time=-0.0046, value=0.4}
          {arriveTangent=3.9833, interpMode=RCIM_Cubic, leaveTangent=3.9833, tangentMode=RCTM_User, time=0.4021, value=2.4036}
          <at 1.0011 3.4809 RCIM_Cubic>
         ]}}})))
   &CameraOffset(Camera Offset Amount=30)
   &ScaleColor(
    Scale Alpha=&FloatFromCurve(FloatCurve=DataInterface() {propertyValues {Curve {keys [<at 0 1 RCIM_Cubic>,<at 1.0011 0.002>]}}})
    ScaleA=true
    ScaleRGB=false
   )
   &DynamicMaterialParameters(
    Index 0 Param 1=1
    Module.Param0WriteEnabled=true
    Module.Param1WriteEnabled=false
    Module.Param2WriteEnabled=false
    Module.Param3WriteEnabled=false
   )
  ]
  rendererClasses [&NiagaraSpriteRendererProperties]
  renderers [{rendererClass=&NiagaraSpriteRendererProperties}]
 }
 SmokeWorld {
  emitterUpdateScript [
   &EmitterState()
   &SpawnBurst_Instantaneous(
    Spawn Count=&Multiply_Int(A=&MakeCustomIntFromBool(Bool=$User.bIsInWater, True INT=0, False INT=1), B=20)
    Use Spawn Probability=false
   ) {module=SpawnBurst_Instantaneous001}
  ]
  particleSpawnScript [
   &InitializeParticle(
    Color=<rgba 0.9062 0.9062 0.9062 0.1>
    Lifetime=&UniformRangedFloat(Minimum=0.4, Maximum=0.75)
    Module.Write Color=true
    Module.Write Lifetime=true
    Module.Write Mass=true
    Module.Write Material Random=true
    Module.Write Position=false
    Module.Write Scale=false
    Module.Write SpriteRotation=true
    Module.Write SpriteSize=true
    Position=&SimulationPosition()
    Sprite Rotation=&UniformRangedFloat(Minimum=0, Maximum=360)
    Sprite Size=&Vector2DFromFloat(Value=&UniformRangedFloat(Minimum=75, Maximum=100))
   )
   &TorusLocation(
    Align Torus With Axis=true
    Large Radius=5.5
    MaskSpawnGroup=false
    Radius Position=1.2
    Torus Distribution Mode:ENiagaraTorusDistributionMode=::NewEnumerator1
    Torus Mode=::NewEnumerator2
    U Position:NiagaraFloat=&Divide_Float(A=/Niagara/DynamicInputs/Execution/ReturnParticleID_AsFloat(), B=20)
   )
   &AddVelocityFromPoint(
    Module.Use Falloff Distance=false
    Velocity Origin:Vector3f=&SimulationPosition()
    Velocity Strength=&UniformRangedFloat(Minimum=750, Maximum=950)
   )
   &ApplyInitialForces()
  ]
  particleUpdateScript [
   &ParticleState()
   &AccelerationForce(Acceleration=&MakeVector(X=0, Y=0, Z=&UniformRangedFloat(Minimum=50, Maximum=200)))
   &Drag(Drag=5, Use Linear Drag=true, Use Rotational Drag=false)
   &SolveForcesAndVelocity(Clamp Velocity=false, Limit Acceleration=false)
   &ScaleSpriteSize(Scale Factor=&Vector2DFromFloat(Value=&FloatFromCurve(FloatCurve=DataInterface() {propertyValues {Curve {keys [
          {arriveTangent=7.1593, interpMode=RCIM_Cubic, leaveTangent=7.1593, tangentMode=RCTM_User, time=0, value=0}
          {arriveTangent=1.1699, interpMode=RCIM_Cubic, leaveTangent=1.1699, tangentMode=RCTM_User, time=0.31, value=0.97}
          <at 1 1.25 RCIM_Cubic>
         ]}}})))
   /SetVariables(Particles.SubImageIndex=&FloatFromCurve(FloatCurve=DataInterface() {propertyValues {Curve {keys [
         {arriveTangent=29.8623, interpMode=RCIM_Cubic, leaveTangent=29.8623, tangentMode=RCTM_User, time=-0.0046, value=5.0556}
         {arriveTangent=8.1241, interpMode=RCIM_Cubic, leaveTangent=8.1241, tangentMode=RCTM_User, time=1.0011, value=22.1689}
        ]}}})) {isSetParameters=true}
   &DynamicMaterialParameters(
    Index 0 Param 1=&FloatFromCurve(FloatCurve=DataInterface() {propertyValues {Curve {keys [<at -0.0046 1.0113>,<at 0.5 1>,<at 1.0011 -1>]}}})
    Index 0 Param 4=0.5
    Module.Param0WriteEnabled=true
    Module.Param1WriteEnabled=false
    Module.Param2WriteEnabled=false
    Module.Param3WriteEnabled=true
   ) {module=DynamicMatParams}
   &ScaleColor(
    Scale Alpha=&FloatFromCurve(FloatCurve=DataInterface() {propertyValues {Curve {keys [<at 0 0 RCIM_Cubic>, {arriveTangent=0.002, interpMode=RCIM_Cubic, leaveTangent=0.002, time=0.5, value=1}, <at 1.0011 0.002 RCIM_Cubic>]}}})
    Scale RGB=&MakeVectorFromLinearColorRGB(Color=&ColorFromCurve(ColorCurve=DataInterface() {propertyValues {
        AlphaCurve {keys [<at 0 1 RCIM_Cubic>,<at 1 0 RCIM_Cubic>]}
        BlueCurve {keys [<at 0 1.14 RCIM_Cubic>,<at 1 1 RCIM_Cubic>]}
        GreenCurve {keys [<at 0 1.842 RCIM_Cubic>,<at 1 1 RCIM_Cubic>]}
        RedCurve {keys [<at 0 3 RCIM_Cubic>,<at 1 1 RCIM_Cubic>]}
       }}))
    ScaleA=true
    ScaleRGB=true
   )
  ]
  rendererClasses [&NiagaraSpriteRendererProperties]
  renderers [{rendererClass=&NiagaraSpriteRendererProperties}]
 }
 Streaks {
  emitterUpdateScript [
   &EmitterState()
   &SpawnBurst_Instantaneous(
    Spawn Count=&Multiply_Int(A=&MakeCustomIntFromBool(Bool=$User.bIsInWater, True INT=0, False INT=1), B=20)
    Use Spawn Probability=false
   ) {module=SpawnBurst_Instantaneous001}
  ]
  particleSpawnScript [
   &InitializeParticle(
    Lifetime=&UniformRangedFloat(Minimum=0.1, Maximum=0.3)
    Module.Write Color=true
    Module.Write Lifetime=true
    Module.Write Mass=true
    Module.Write Material Random=true
    Module.Write Position=false
    Module.Write Scale=false
    Module.Write SpriteRotation=true
    Module.Write SpriteSize=true
    Position=&SimulationPosition()
    Sprite Rotation=&UniformRangedFloat(Minimum=0, Maximum=360)
    Sprite Size=&UniformRangedVector2D(Minimum=<1 5>, Maximum=<3 5>)
   )
   &CylinderLocation(
    Cylinder Height=0
    Cylinder Radius=10
    MaskSpawnGroup=false
    Module.Enable Lathe Profile=false
    Module.Override Randomness=false
    Module.Surface Only=true
    Surface Only Band Thickness=1
    Use Endcaps In Surface Only Mode=false
   )
   &AddVelocity(Velocity=&MakeVector(X=0, Y=0, Z=&UniformRangedFloat(Minimum=0, Maximum=200)))
   &AddVelocityFromPoint(Module.Use Falloff Distance=false, Velocity Origin:Vector3f=&SimulationPosition(), Velocity Strength=1100)
   &ApplyInitialForces()
  ]
  particleUpdateScript [
   &ParticleState()
   &Drag(Drag=5, Use Linear Drag=true, Use Rotational Drag=false)
   &SolveForcesAndVelocity(Clamp Velocity=false, Limit Acceleration=false)
   &ScaleColor(
    Scale Alpha=&FloatFromCurve(FloatCurve=DataInterface() {propertyValues {Curve {keys [<at 0 0 RCIM_Cubic>, {arriveTangent=0.002, interpMode=RCIM_Cubic, leaveTangent=0.002, time=0.5, value=1}, <at 1.0011 0.002 RCIM_Cubic>]}}})
    ScaleA=true
    ScaleRGB=false
   )
   &ScaleSpriteSizeBySpeed(Max Scale Factor=<1 10>)
  ]
  rendererClasses [&NiagaraSpriteRendererProperties]
  renderers [{rendererClass=&NiagaraSpriteRendererProperties}]
 }
 Water_splash_arms {
  emitterUpdateScript [
   &EmitterState()
   &SpawnBurst_Instantaneous(
    Spawn Count=&Multiply_Int(A=&MakeCustomIntFromBool(Bool=$User.bIsInWater, True INT=1, False INT=0), B=13)
    Use Spawn Probability=false
   )
  ]
  particleSpawnScript [
   &InitializeParticle(
    Lifetime=&Multiply_Float(A=&UniformRangedFloat(Minimum=0.5, Maximum=0.7), B=$User.EffectScale)
    Mass=&UniformRangedFloat(Minimum=0.6, Maximum=1.2)
    Mesh Scale=&UniformRangedVector(Minimum=<0.1 0.15 0.1>, Maximum=<0.3 0.6 0.45>)
    Module.Write Color=true
    Module.Write Lifetime=true
    Module.Write Mass=true
    Module.Write Material Random=true
    Module.Write Position=false
    Module.Write Scale=true
    Module.Write SpriteRotation=true
    Module.Write SpriteSize=false
    Position=&SimulationPosition()
    Sprite Rotation=&NormalizedAngleToDegrees(NormalizedAngle=&UniformRangedFloat(Minimum=0, Maximum=1))
   )
   &InitializeWaterSnapToOwner()
   &ConstrainPositionToPlane(
    Plane Normal:Vector3f=$Particles.WaterPlaneNormal
    Plane Position:NiagaraPosition=&Add_Vector(A=<0 0 10>, B=$Particles.WaterPlaneLocation)
   )
   &TorusLocation(Align Torus With Axis=false, Large Radius=25, MaskSpawnGroup=false, Module.Override Randomness=false, Torus Mode=::NewEnumerator2)
   &PointForce(Force Origin=&SimulationPosition(), Force Strength=50, Module.Use Falloff=false, Module.UseLinearForce=false)
   &AccelerationForce(Acceleration=<0 0 -50>)
   &ApplyInitialForces()
   /Niagara/Modules/Spawn/Orientation/V2/InitialMeshOrientation(Orientation Vector=$Particles.Velocity, Use Orientation Vector=true, Use Rotation Vector=false)
  ]
  particleUpdateScript [
   &ParticleState() {module=UpdateAge}
   &DynamicMaterialParameters(
    Index 0 Param 1=&FloatFromCurve(FloatCurve=DataInterface() {propertyValues {Curve {keys [
         {arriveTangent=0.8954, interpMode=RCIM_Cubic, leaveTangent=0.8954, tangentMode=RCTM_User, time=-0.0005, value=-1.845}
         {arriveTangent=1.8419, interpMode=RCIM_Cubic, leaveTangent=1.8419, time=0.7887, value=-0.7711}
         {arriveTangent=3.1029, interpMode=RCIM_Cubic, leaveTangent=3.1029, tangentMode=RCTM_User, time=1.0011, value=0}
        ]}}})
    Index 0 Param 2=&FloatFromCurve(FloatCurve=DataInterface() {propertyValues {Curve {keys [<at -0.0046 4>,<at 1.0011 0>]}}})
    Index 0 Param 3=&FloatFromCurve(
     FloatCurve=DataInterface() {propertyValues {Curve {keys [
         {arriveTangent=0.096, interpMode=RCIM_Cubic, leaveTangent=0.096, tangentMode=RCTM_User, time=-0.0046, value=0.0882}
         <at 1.0011 0.1372 RCIM_Cubic>
        ]}}}
     Scale Curve=&Multiply_Float(A=2, B=$User.EffectScale)
    )
    Index 0 Param 4=&FloatFromCurve(FloatCurve=DataInterface() {propertyValues {Curve {keys [
         {arriveTangent=1.9125, interpMode=RCIM_Cubic, leaveTangent=0.2728, tangentMode=RCTM_User, time=-0.0046, value=0.9276}
         {arriveTangent=4.6326, interpMode=RCIM_Cubic, leaveTangent=4.6326, tangentMode=RCTM_User, time=1.0011, value=2.7648}
        ]}}})
    Module.Param0WriteEnabled=true
    Module.Param1WriteEnabled=true
    Module.Param2WriteEnabled=true
    Module.Param3WriteEnabled=true
   ) {module=DynamicMaterialParameters001}
   &ScaleColor(
    Scale Alpha=&FloatFromCurve(FloatCurve=DataInterface() {propertyValues {Curve {keys [<at -0.0046 1.0113 RCIM_Cubic>,<at 1.0011 0.002 RCIM_Cubic>]}}})
    ScaleA=true
    ScaleRGB=false
   )
   &Drag(Drag=2, Use Linear Drag=true, Use Rotational Drag=false)
   &SolveForcesAndVelocity(Clamp Velocity=false, Limit Acceleration=false)
   &ScaleMeshSize(Scale Factor=&VectorFromCurve(VectorCurve=DataInterface() {propertyValues {
       XCurve {keys [<at -0.0011 1.0067>,<at 0.9989 3>]}
       YCurve {keys [<at -0.0011 1.0067>,<at 0.9989 3>]}
       ZCurve {keys [<at -0.0011 1.1211>,<at 0.9989 1.5622>]}
      }})) {module=MeshSizeScale}
  ]
  rendererClasses [&NiagaraMeshRendererProperties]
  renderers [{rendererClass=&NiagaraMeshRendererProperties, rendererIndex=0}]
 }
 Water_splash_up {
  bEnabled=false
  emitterUpdateScript [
   &EmitterState()
   &SpawnBurst_Instantaneous(
    Spawn Count=&Multiply_Int(A=&MakeCustomIntFromBool(Bool=$User.bIsInWater, True INT=1, False INT=0), B=20)
    Use Spawn Probability=false
   )
   &SpawnBurst_Instantaneous(Spawn Count=6, Use Spawn Probability=false) {enabled=false, module=SpawnBurst_Instantaneous001}
  ]
  particleSpawnScript [
   &InitializeParticle(
    Color=<rgba 1 1 1 0.5>
    Lifetime=&UniformRangedFloat(Minimum=0.4, Maximum=0.55)
    Mass=&UniformRangedFloat(Minimum=0.6, Maximum=1.2)
    Mesh Scale=&UniformRangedVector(Minimum=<0.1 0.15 0.1>, Maximum=<0.2 0.4 0.25>)
    Module.Write Color=true
    Module.Write Lifetime=true
    Module.Write Mass=true
    Module.Write Material Random=true
    Module.Write Position=false
    Module.Write Scale=true
    Module.Write SpriteRotation=true
    Module.Write SpriteSize=false
    Position=&SimulationPosition()
    Sprite Rotation=&NormalizedAngleToDegrees(NormalizedAngle=&UniformRangedFloat(Minimum=0, Maximum=1))
   )
   &InitializeWaterSnapToOwner()
   &ConstrainPositionToPlane(
    Plane Normal:Vector3f=$Particles.WaterPlaneNormal
    Plane Position:NiagaraPosition=&Add_Vector(A=<0 0 10>, B=$Particles.WaterPlaneLocation)
   )
   &TorusLocation(
    Align Torus With Axis=true
    Large Radius=2
    MaskSpawnGroup=false
    Radius Position=1.2
    Torus Distribution Mode:ENiagaraTorusDistributionMode=::NewEnumerator1
    Torus Mode=::NewEnumerator2
    U Position:NiagaraFloat=&Divide_Float(A=&ReturnParticleID_AsFloat(), B=20)
   ) {enabled=false}
   &TorusLocation(Align Torus With Axis=false, Large Radius=40, MaskSpawnGroup=false, Module.Override Randomness=false, Torus Mode=::NewEnumerator2) {module=TorusLocation001}
   &AddVelocityFromPoint(
    Module.Use Falloff Distance=false
    Velocity Origin:Vector3f=&SimulationPosition()
    Velocity Strength=&UniformRangedFloat(Minimum=10, Maximum=30)
   )
   &PointForce(Force Origin=&SimulationPosition(), Force Strength=50, Module.Use Falloff=false, Module.UseLinearForce=false) {enabled=false}
   &ApplyInitialForces()
   &InitialMeshOrientation(Orientation Vector=$Particles.Velocity, Use Orientation Vector=true, Use Rotation Vector=false)
   /SetVariables(Particles.WPOWidth:NiagaraFloat=&UniformRangedFloat(Minimum=0.2, Maximum=0.3)) {isSetParameters=true}
  ]
  particleUpdateScript [
   &ParticleState() {module=UpdateAge}
   &ScaleColor(
    Scale Alpha=&FloatFromCurve(FloatCurve=DataInterface() {propertyValues {Curve {keys [<at -0.0046 1.0113 RCIM_Cubic>,<at 1.0011 0.002 RCIM_Cubic>]}}})
    ScaleA=true
    ScaleRGB=false
   )
   &Drag(Drag=2, Use Linear Drag=true, Use Rotational Drag=false)
   &SolveForcesAndVelocity(Clamp Velocity=false, Limit Acceleration=false)
   &DynamicMaterialParameters(
    Index 0 Param 1=&FloatFromCurve(FloatCurve=DataInterface() {propertyValues {Curve {keys [<at 0 -2>,<at 1 -1.5>]}}})
    Index 0 Param 2=1
    Index 0 Param 3=&FloatFromCurve(FloatCurve=DataInterface() {propertyValues {Curve {keys [<at 0 0.4798>,<at 1 1>]}}}, Scale Curve=$Particles.WPOWidth)
    Index 0 Param 4=&FloatFromCurve(FloatCurve=DataInterface() {propertyValues {Curve {keys [<at 0 -1>,<at 1 0>]}}})
    Index 1 Param 1=&FloatFromCurve(FloatCurve=DataInterface() {propertyValues {Curve {keys [
         <at 0 0 RCIM_Cubic>
         {arriveTangent=1, interpMode=RCIM_Cubic, leaveTangent=1, time=0.8512, value=0.2958}
         {arriveTangent=4.767, interpMode=RCIM_Cubic, leaveTangent=4.7671, tangentMode=RCTM_User, time=1, value=1}
        ]}}})
    Module.Param0WriteEnabled=true
    Module.Param0WriteEnabled1=true
    Module.Param1WriteEnabled=true
    Module.Param1WriteEnabled1=true
    Module.Param2WriteEnabled=true
    Module.Param2WriteEnabled1=true
    Module.Param3WriteEnabled=true
    Module.Param3WriteEnabled1=true
    Write Parameter Index 1=true
   ) {module=DynamicMaterialParameters001}
   &ScaleMeshSize(Scale Factor=&VectorFromCurve(VectorCurve=DataInterface() {propertyValues {
       XCurve {keys [<at -0.0011 0.25>,<at 0.9988 1.5>]}
       YCurve {keys [<at -0.0011 0.25>,<at 0.9988 1.5>]}
       ZCurve {keys [<at -0.0011 0.3>,<at 0.9988 0.5>]}
      }}))
  ]
  rendererClasses [&NiagaraMeshRendererProperties]
  renderers [{rendererClass=&NiagaraMeshRendererProperties, rendererIndex=0}]
 }
}
summary {
 systemName=NS_Emote_Superhero_Backflip
 emitters [
  {emitterName=FN_Water_Drops, rendererClasses [&NiagaraSpriteRendererProperties]}
  {bEnabled=false, emitterName=FN_Splash, rendererClasses [&NiagaraSpriteRendererProperties]}
  {emitterName=SmokeWorld, rendererClasses [&NiagaraSpriteRendererProperties]}
  {bEnabled=false, emitterName=FN_Debug, rendererClasses [&NiagaraSpriteRendererProperties]}
  {emitterName=FN_Water_Foam, rendererClasses [&NiagaraMeshRendererProperties]}
  {emitterName=SmokeGround, rendererClasses [&NiagaraSpriteRendererProperties]}
  {emitterName=Streaks, rendererClasses [&NiagaraSpriteRendererProperties]}
  {bEnabled=false, emitterName=Water_splash_up, rendererClasses [&NiagaraMeshRendererProperties]}
  {emitterName=Water_splash_arms, rendererClasses [&NiagaraMeshRendererProperties]}
 ]
 userVariables [
  User.bIsInWater {type=NiagaraBool}
  User.Boost {type=NiagaraFloat, defaultValue=0, description=""}
  User.EffectScale {type=NiagaraFloat, defaultValue=1, description=""}
  User.EffectSpawnScale {type=NiagaraFloat, defaultValue=10, description=""}
  User.ForwardSpeed {type=NiagaraFloat, defaultValue=0, description=""}
  User.FrontPontoonBig.Depth {type=NiagaraFloat, defaultValue=0, description=""}
  User.FrontPontoonBig.Impulse {type=Vector3f, description="", defaultValue {struct=Vector3f, value=<0 0 0>}}
  User.InWater {type=NiagaraBool}
  User.isBoosting {type=NiagaraBool}
  User.MidPontoonBig.Depth {type=NiagaraFloat, defaultValue=0, description=""}
  User.MidPontoonBig.Impulse {type=Vector3f, description="", defaultValue {struct=Vector3f, value=<0 0 0>}}
  User.NewVariable {type=NiagaraFloat, defaultValue=0, description=""}
  User.RearPontoonBig.Depth {type=NiagaraFloat, defaultValue=0, description=""}
  User.RearPontoonBig.Impulse {type=Vector3f, description="", defaultValue {struct=Vector3f, value=<0 0 0>}}
  User.WaterBodyIndex {type=NiagaraFloat, defaultValue=0, description=""}
  User.WaterDepth {type=NiagaraFloat, defaultValue=0, description=""}
  User.WaterPlaneLocation {type=NiagaraPosition, description="", defaultValue {}}
  User.WaterPlaneNormal {type=Vector3f, description="", defaultValue {struct=Vector3f, value=<0 0 0>}}
  User.WaveAttenuationFactor {type=NiagaraFloat, defaultValue=0, description=""}
  User.WaveTime {type=NiagaraFloat, defaultValue=0, description=""}
 ]
}
