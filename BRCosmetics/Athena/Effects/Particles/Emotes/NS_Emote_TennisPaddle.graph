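# Niagara one-shot puff effect for the tennis paddle emote, made of short-lived smoke and wind sprites.
# SmokeLinear bursts gray puffs from a torus along a downward, sideways velocity, while SmokeRadial pushes puffs away from a point, both using curve-driven drag, size, sub-image and alpha fade.
# Wind_Boof adds a chance-based, growing sprite with fixed vertical facing.
# Two mesh smoke emitters and a spike streak mesh emitter are disabled.
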NiagaraSystem /BRCosmetics/Athena/Effects/Particles/Emotes/NS_Emote_TennisPaddle
dependencies {
 usedDataInterfaces [&NiagaraDataInterfaceCurve, &NiagaraDataInterfaceVectorCurve]
 usedDynamicInputs []
 usedModules [&SystemState]
}
emitters {
 SmokeLinear {
  emitterUpdateScript [
   &EmitterState()
   &SpawnBurst_Instantaneous(Spawn Count=12, Spawn Group=1, Use Spawn Probability=false)
  ]
  particleSpawnScript [
   /Niagara/Modules/Spawn/Initialization/InitializeParticle(
    Color=&UniformRangedLinearColor(Minimum=<rgba 0.2552 0.2552 0.2552 0.15>, Maximum=<rgba 0.5255 0.5255 0.5255 0.3>)
    Lifetime=&UniformRangedFloat(Minimum=0.1, Maximum=0.25)
    Mesh Scale=&VectorFromFloat(Value=&UniformRangedFloat(Minimum=0.01, Maximum=0.05))
    Module.Write Color=true
    Module.Write Lifetime=true
    Module.Write Mass=true
    Module.Write Material Random=true
    Module.Write Position=true
    Module.Write Scale=false
    Module.Write SpriteRotation=false
    Module.Write SpriteSize=true
    Position=&SimulationPosition()
    Sprite Size=&Vector2DFromFloat(Value=&UniformRangedFloat(Minimum=1, Maximum=3))
   )
   &TorusLocation(
    Align Torus With Axis=false
    Large Radius=2
    MaskSpawnGroup=false
    Module.Override Randomness=false
    Offset=<0 0 -1>
    Torus Mode=::NewEnumerator2
   )
   &AddVelocity(Velocity=&Multiply_Vector(
     A=&UniformRangedVector(Minimum=<-10 90 -200>, Maximum=<10 110 -200>)
     B=&VectorFromFloat(Value=&UniformRangedFloat(Minimum=0.75, Maximum=1.5))
    ))
  ]
  particleUpdateScript [
   &Drag(
    Drag=&FloatFromCurve(
     FloatCurve=DataInterface() {propertyValues {Curve {keys [{arriveTangent=7.1572, interpMode=RCIM_Cubic, leaveTangent=7.1573, tangentMode=RCTM_User, time=-0.0053, value=0.1532}, <at 0.3232 1>]}}}
     Scale Curve=10
    )
    Use Linear Drag=true
    Use Rotational Drag=false
   )
   &SolveForcesAndVelocity(Clamp Velocity=false, Limit Acceleration=false)
   &ParticleState()
   &ScaleSpriteSize(Scale Factor=&Vector2DFromFloat(Value=&FloatFromCurve(FloatCurve=DataInterface() {propertyValues {Curve {keys [
          {arriveTangent=7.8209, interpMode=RCIM_Cubic, leaveTangent=7.821, tangentMode=RCTM_User, time=0, value=2}
          {arriveTangent=-1.4983, interpMode=RCIM_Cubic, leaveTangent=-1.4983, time=0.3044, value=3}
          {arriveTangent=-2.8451, interpMode=RCIM_Cubic, leaveTangent=-2.8451, tangentMode=RCTM_User, time=1.0011, value=0.5}
         ]}}})))
   /SetVariables(Particles.SubImageIndex=&FloatFromCurve(FloatCurve=DataInterface() {propertyValues {Curve {keys [
         {arriveTangent=29.8623, interpMode=RCIM_Cubic, leaveTangent=29.8623, tangentMode=RCTM_User, time=-0.0046, value=5.0556}
         <at 1.0011 22.1689 RCIM_Cubic>
        ]}}})) {isSetParameters=true, module=SetVariables_#1}
   &DynamicMaterialParameters(
    Index 0 Param 1=&FloatFromCurve(FloatCurve=DataInterface() {propertyValues {Curve {keys [<at -0.0046 1.0113 RCIM_Cubic>,<at 1.0011 0.002>]}}})
    Index 0 Param 4=&FloatFromCurve(FloatCurve=DataInterface() {propertyValues {Curve {keys [<at 0 0 RCIM_Cubic>,<at 1 0.5>]}}})
    Module.Param0WriteEnabled=true
    Module.Param1WriteEnabled=false
    Module.Param2WriteEnabled=false
    Module.Param3WriteEnabled=true
   ) {module=DynamicMatParams}
   &ScaleColor(
    Scale Alpha=&FloatFromCurve(FloatCurve=DataInterface() {propertyValues {Curve {keys [
         {arriveTangent=-3.1757, interpMode=RCIM_Cubic, leaveTangent=-3.1757, tangentMode=RCTM_User, time=0.5, value=1.0113}
         {arriveTangent=-1.5121, interpMode=RCIM_Cubic, leaveTangent=-1.5121, tangentMode=RCTM_User, time=1.0011, value=0.002}
        ]}}})
    ScaleA=true
    ScaleRGB=false
   )
  ]
  rendererClasses [&NiagaraSpriteRendererProperties]
  renderers [{rendererClass=&NiagaraSpriteRendererProperties}]
 }
 SmokeRadial {
  emitterUpdateScript [
   &EmitterState()
   &SpawnBurst_Instantaneous(Spawn Count=8, Spawn Group=1, Use Spawn Probability=false)
  ]
  particleSpawnScript [
   &InitializeParticle(
    Color=&UniformRangedLinearColor(Minimum=<rgba 0.2552 0.2552 0.2552 0.125>, Maximum=<rgba 0.5255 0.5255 0.5255 0.4>)
    Lifetime=&UniformRangedFloat(Minimum=0.15, Maximum=0.3)
    Mesh Scale=&VectorFromFloat(Value=&UniformRangedFloat(Minimum=0.01, Maximum=0.05))
    Module.Write Color=true
    Module.Write Lifetime=true
    Module.Write Mass=true
    Module.Write Material Random=true
    Module.Write Position=true
    Module.Write Scale=false
    Module.Write SpriteRotation=false
    Module.Write SpriteSize=true
    Position=&SimulationPosition()
    Sprite Size=&Vector2DFromFloat(Value=&UniformRangedFloat(Minimum=1, Maximum=3))
   )
   &SystemLocation(MaskSpawnGroup=false, Offset=<0 0 2>)
   &TorusLocation(
    Align Torus With Axis=false
    Large Radius=5
    MaskSpawnGroup=false
    Module.Override Randomness=false
    Offset=<0 0 -1>
    Torus Mode=::NewEnumerator2
   )
   &AddVelocityFromPoint(
    Module.Use Falloff Distance=false
    Origin Offset=<0 -1 3>
    Velocity Origin:Vector3f=&SimulationPosition()
    Velocity Strength=&UniformAOrBFloat(A=50, B=75)
   )
  ]
  particleUpdateScript [
   &ParticleState()
   &Drag(
    Drag=&FloatFromCurve(
     FloatCurve=DataInterface() {propertyValues {Curve {keys [{arriveTangent=7.1572, interpMode=RCIM_Cubic, leaveTangent=7.1573, tangentMode=RCTM_User, time=-0.0053, value=0.1532}, <at 0.3232 1>]}}}
     Scale Curve=10
    )
    Use Linear Drag=true
    Use Rotational Drag=false
   )
   &SolveForcesAndVelocity(Clamp Velocity=false, Limit Acceleration=false)
   &ScaleSpriteSize(Scale Factor=&Vector2DFromFloat(Value=&FloatFromCurve(FloatCurve=DataInterface() {propertyValues {Curve {keys [
          {arriveTangent=4.6575, interpMode=RCIM_Cubic, leaveTangent=4.6575, tangentMode=RCTM_User, time=-0.0046, value=1.1969}
          <at 1.0011 3.4809 RCIM_Cubic>
         ]}}})))
   /SetVariables(Particles.SubImageIndex=&FloatFromCurve(FloatCurve=DataInterface() {propertyValues {Curve {keys [
         {arriveTangent=29.8623, interpMode=RCIM_Cubic, leaveTangent=29.8623, tangentMode=RCTM_User, time=-0.0046, value=5.0556}
         <at 1.0011 22.1689 RCIM_Cubic>
        ]}}})) {isSetParameters=true, module=SetVariables_#1}
   &DynamicMaterialParameters(
    Index 0 Param 1=&FloatFromCurve(FloatCurve=DataInterface() {propertyValues {Curve {keys [<at -0.0046 1.0113 RCIM_Cubic>,<at 1.0011 0.002>]}}})
    Index 0 Param 4=&FloatFromCurve(FloatCurve=DataInterface() {propertyValues {Curve {keys [<at 0 0 RCIM_Cubic>,<at 1 0.5>]}}})
    Module.Param0WriteEnabled=true
    Module.Param1WriteEnabled=false
    Module.Param2WriteEnabled=false
    Module.Param3WriteEnabled=true
   ) {module=DynamicMatParams}
   &ScaleColor(
    Scale Alpha=&FloatFromCurve(FloatCurve=DataInterface() {propertyValues {Curve {keys [
         {arriveTangent=-3.1757, interpMode=RCIM_Cubic, leaveTangent=-3.1757, tangentMode=RCTM_User, time=0.5, value=1.0113}
         {arriveTangent=-1.5121, interpMode=RCIM_Cubic, leaveTangent=-1.5121, tangentMode=RCTM_User, time=1.0011, value=0.002}
        ]}}})
    ScaleA=true
    ScaleRGB=false
   )
  ]
  rendererClasses [&NiagaraSpriteRendererProperties]
  renderers [{rendererClass=&NiagaraSpriteRendererProperties}]
 }
 Smoke_Linear {
  bEnabled=false
  emitterUpdateScript [
   &EmitterScalability(
    Module.bApplySpawnCountScale=false
    Module.bApplySpawnCountScaleByDistanceFraction=false
    Module.bUseMaxDistance=false
    Module.bUseMinDistance=false
   ) {enabled=false}
   &EmitterLifeCycle(MaxLoopCount=1, NextLoopDuration=1) {enabled=false}
   &EmitterState()
   &SpawnRate(SpawnRate=50, Use Spawn Probability=false) {enabled=false}
   &SpawnBurst_Instantaneous(Spawn Count=2, Spawn Group=1, Use Spawn Probability=false) {enabled=false}
  ]
  particleSpawnScript [
   &InitializeParticle(
    Lifetime=&UniformRangedFloat(Minimum=0.25, Maximum=0.3)
    Mesh Scale=&VectorFromFloat(Value=&UniformRangedFloat(Minimum=0.01, Maximum=0.05))
    Module.Write Color=true
    Module.Write Lifetime=true
    Module.Write Mass=true
    Module.Write Material Random=true
    Module.Write Position=true
    Module.Write Scale=true
    Module.Write SpriteRotation=false
    Module.Write SpriteSize=false
    Position=&SimulationPosition()
    Sprite Size=<25 25>
   )
   /Niagara/Modules/Spawn/Orientation/V2/InitialMeshOrientation(
    Rotation=&UniformRangedVector(Minimum=<0 0 0>, Maximum=<1 1 1>)
    Use Orientation Vector=true
    Use Rotation Vector=true
   )
   &DynamicMaterialParameters(
    Index 0 Param 1=&UniformRangedFloat(Minimum=1.25, Maximum=1)
    Index 0 Param 2=0.1
    Index 0 Param 4=0.05
    Module.Param0WriteEnabled=true
    Module.Param1WriteEnabled=true
    Module.Param2WriteEnabled=false
    Module.Param3WriteEnabled=true
   )
   &AddVelocity(Velocity=&UniformRangedVector(Minimum=<0 150 -250>, Maximum=<1 150 -100>))
  ]
  particleUpdateScript [
   &Drag(
    Drag=&FloatFromCurve(
     FloatCurve=DataInterface() {propertyValues {Curve {keys [{arriveTangent=7.1572, interpMode=RCIM_Cubic, leaveTangent=7.1573, tangentMode=RCTM_User, time=-0.0053, value=0.1532}, <at 0.3232 1>]}}}
     Scale Curve=10
    )
    Use Linear Drag=true
    Use Rotational Drag=false
   )
   &SolveForcesAndVelocity(Clamp Velocity=false, Limit Acceleration=false)
   &ParticleState()
   &ScaleColor(
    Scale Alpha=&FloatFromCurve(FloatCurve=DataInterface() {propertyValues {Curve {keys [
         <at -0.0046 0.8475 RCIM_Cubic>
         {arriveTangent=-1.3398, interpMode=RCIM_Cubic, leaveTangent=-1.3398, time=0.4274, value=0.5912}
         <at 1.0011 -0.5 RCIM_Cubic>
        ]}}})
    Scale RGB=&MakeVectorFromLinearColorRGB(Color=/Niagara/DynamicInputs/TypeConversions/MakeLinearColorFromFloat(Value:NiagaraFloat=0.05))
    ScaleA=true
    ScaleRGB=true
   )
   &ScaleMeshSize(Scale Factor=&VectorFromFloat(Value=&FloatFromCurve(
      FloatCurve=DataInterface() {propertyValues {Curve {keys [
          {arriveTangent=4.4176, interpMode=RCIM_Cubic, leaveTangent=4.4176, tangentMode=RCTM_User, time=0.001, value=0.5}
          <at 0.12 1.5325>
          {arriveTangent=-0.8233, interpMode=RCIM_Cubic, leaveTangent=-0.8233, time=0.3069, value=0.8913}
          <at 1.0009 0.8072>
         ]}}}
      Scale Curve=0.2
     )))
  ]
  rendererClasses [&NiagaraMeshRendererProperties]
  renderers [{rendererClass=&NiagaraMeshRendererProperties, rendererIndex=0}]
 }
 Smoke_RadialLocal {
  bEnabled=false
  emitterUpdateScript [
   &EmitterScalability(
    Module.bApplySpawnCountScale=false
    Module.bApplySpawnCountScaleByDistanceFraction=false
    Module.bUseMaxDistance=false
    Module.bUseMinDistance=false
   ) {enabled=false}
   &EmitterLifeCycle(MaxLoopCount=1, NextLoopDuration=1) {enabled=false}
   &EmitterState()
   &SpawnRate(SpawnRate=50, Use Spawn Probability=false) {enabled=false}
   &SpawnBurst_Instantaneous(Spawn Count=6, Spawn Group=1, Use Spawn Probability=false)
  ]
  particleSpawnScript [
   &InitializeParticle(
    Lifetime=&UniformRangedFloat(Minimum=0.25, Maximum=0.3)
    Mesh Scale=&VectorFromFloat(Value=&UniformRangedFloat(Minimum=0.01, Maximum=0.05))
    Module.Write Color=true
    Module.Write Lifetime=true
    Module.Write Mass=true
    Module.Write Material Random=true
    Module.Write Position=true
    Module.Write Scale=true
    Module.Write SpriteRotation=false
    Module.Write SpriteSize=false
    Position=&SimulationPosition()
    Sprite Size=<25 25>
   )
   &InitialMeshOrientation(
    Rotation=&UniformRangedVector(Minimum=<0 0 0>, Maximum=<1 1 1>)
    Use Orientation Vector=true
    Use Rotation Vector=true
   )
   &DynamicMaterialParameters(
    Index 0 Param 1=&UniformRangedFloat(Minimum=1.25, Maximum=1)
    Index 0 Param 2=0.1
    Index 0 Param 4=0.05
    Module.Param0WriteEnabled=true
    Module.Param1WriteEnabled=true
    Module.Param2WriteEnabled=false
    Module.Param3WriteEnabled=true
   )
   &TorusLocation(
    Align Torus With Axis=false
    Large Radius=2.5
    MaskSpawnGroup=false
    Module.Override Randomness=false
    Offset=<0 0 -1>
    Torus Mode=::NewEnumerator2
   )
   &AddVelocityFromPoint(
    Module.Use Falloff Distance=false
    Origin Offset=<0 -1 3>
    Velocity Origin:Vector3f=&SimulationPosition()
    Velocity Strength=&UniformAOrBFloat(A=50, B=75)
   )
  ]
  particleUpdateScript [
   &Drag(
    Drag=&FloatFromCurve(
     FloatCurve=DataInterface() {propertyValues {Curve {keys [{arriveTangent=7.1572, interpMode=RCIM_Cubic, leaveTangent=7.1573, tangentMode=RCTM_User, time=-0.0053, value=0.1532}, <at 0.3232 1>]}}}
     Scale Curve=10
    )
    Use Linear Drag=true
    Use Rotational Drag=false
   )
   &SolveForcesAndVelocity(Clamp Velocity=false, Limit Acceleration=false)
   &ParticleState()
   &ScaleColor(
    Scale Alpha=&FloatFromCurve(FloatCurve=DataInterface() {propertyValues {Curve {keys [
         <at -0.0046 0.8475 RCIM_Cubic>
         {arriveTangent=-1.3398, interpMode=RCIM_Cubic, leaveTangent=-1.3398, time=0.4274, value=0.5912}
         <at 1.0011 -0.5 RCIM_Cubic>
        ]}}})
    Scale RGB=&MakeVectorFromLinearColorRGB(Color=&MakeLinearColorFromFloat(Value:NiagaraFloat=0.05))
    ScaleA=true
    ScaleRGB=true
   )
   &ScaleMeshSize(Scale Factor=&VectorFromFloat(Value=&FloatFromCurve(
      FloatCurve=DataInterface() {propertyValues {Curve {keys [
          {arriveTangent=4.4176, interpMode=RCIM_Cubic, leaveTangent=4.4176, tangentMode=RCTM_User, time=0.001, value=0.5}
          <at 0.12 1.5325>
          {arriveTangent=-0.8233, interpMode=RCIM_Cubic, leaveTangent=-0.8233, time=0.3069, value=0.8913}
          <at 1.0009 0.8072>
         ]}}}
      Scale Curve=0.2
     )))
  ]
  rendererClasses [&NiagaraMeshRendererProperties]
  renderers [{rendererClass=&NiagaraMeshRendererProperties, rendererIndex=0}]
 }
 SpikeStreakMesh_NEW {
  bEnabled=false
  emitterUpdateScript [
   &EmitterState()
   &SpawnBurst_Instantaneous(Spawn Count=&UniformRangedInt(Minimum=3, Maximum=7), Use Spawn Probability=false)
  ]
  particleSpawnScript [
   &InitializeParticle(
    Lifetime=&UniformRangedFloat(Minimum=0.06, Maximum=0.1)
    Mesh Scale=&Multiply_VectorByFloat(Vector=&UniformRangedVector(Minimum=<0.1 0.025 0.025>, Maximum=<0.75 0.05 0.05>), Float=0.6)
    Module.Write Color=true
    Module.Write Lifetime=true
    Module.Write Mass=true
    Module.Write Material Random=true
    Module.Write Position=true
    Module.Write Scale=true
    Module.Write SpriteRotation=false
    Module.Write SpriteSize=false
    Position=&SimulationPosition()
   )
   &TorusLocation(
    Align Torus With Axis=true
    Large Radius=2.5
    MaskSpawnGroup=false
    Module.Override Randomness=false
    Torus Axis=<0 1 -1>
    Torus Mode=::NewEnumerator2
   )
   /Niagara/Modules/Update/Orientation/V2/OrientMeshToVector(
    Look At Direction:Vector3f=&UniformRangedVector(Minimum=<-1 -1 1>, Maximum=<1 3 -3>)
    Override Quat=false
    UseOptionalSideAxis=false
    UseOptionalUpAxis=false
   ) {enabled=false}
   &AddVelocityFromPoint(
    Module.Use Falloff Distance=false
    Origin Offset=<0 -0.5 0>
    Velocity Origin:Vector3f=&SimulationPosition()
    Velocity Strength=&UniformAOrBFloat(A=50, B=75)
   )
   &ApplyInitialForces()
  ]
  particleUpdateScript [
   &ParticleState()
   &Color(
    Scale Alpha=&FloatFromCurve(FloatCurve=DataInterface() {propertyValues {Curve {keys [<at -0.0046 1.0113 RCIM_Cubic>,<at 1.0011 0.002>]}}})
    Scale Color=&VectorFromCurve(VectorCurve=DataInterface() {propertyValues {
       XCurve {keys [<at -0.0031 48.8>,<at 1 1>]}
       YCurve {keys [<at -0.006 20.3>,<at 1 0.8>]}
       ZCurve {keys [<at -0.0011 1.0067>,<at 0.9989 0.28>]}
      }})
   )
   &ScaleMeshSize(Scale Factor=&VectorFromCurve(VectorCurve=DataInterface() {propertyValues {
       bOptimizeLUT=false
       XCurve {keys [{arriveTangent=1.8635, interpMode=RCIM_Cubic, leaveTangent=1.8635, tangentMode=RCTM_User, time=-0.0048, value=0.4}, <at 0.9737 1>]}
       YCurve {keys [
         {arriveTangent=0.1613, interpMode=RCIM_Cubic, leaveTangent=0.1613, tangentMode=RCTM_User, time=0, value=1}
         {arriveTangent=-1.3776, interpMode=RCIM_Cubic, leaveTangent=-1.3776, tangentMode=RCTM_User, time=0.9944, value=0}
        ]}
       ZCurve {keys [<at 0 1 RCIM_Cubic>, {arriveTangent=-1.635, interpMode=RCIM_Cubic, leaveTangent=-1.635, tangentMode=RCTM_User, time=0.9961, value=0}]}
      }}))
   /Game/Effects/Niagara/Modules/Materials/MaterialPanner(Offset:NiagaraFloat=&Add_Float(A=$Emitter.Age, B=$Particles.Initial.MaterialRandom), Speed=-5)
  ]
  rendererClasses [&NiagaraMeshRendererProperties]
  renderers [{rendererClass=&NiagaraMeshRendererProperties, rendererIndex=0}]
 }
 Wind_Boof {
  emitterUpdateScript [
   &EmitterScalability(
    MaxDistance=7000
    Module.bApplySpawnCountScale=false
    Module.bApplySpawnCountScaleByDistanceFraction=false
    Module.bUseMaxDistance=true
    Module.bUseMinDistance=false
   ) {enabled=false}
   &EmitterLifeCycle(MaxLoopCount=1, NextLoopDuration=0.5) {enabled=false}
   &EmitterState()
   &SpawnBurst_Instantaneous(Spawn Count=1, Spawn Probability=0.5, Use Spawn Probability=true)
  ]
  particleSpawnScript [
   &InitializeParticle(
    Lifetime=0.15
    Mesh Scale=&UniformRangedVector(Minimum=<0.2 0.1 0.1>, Maximum=<1 1 1>)
    Module.Write Color=true
    Module.Write Lifetime=true
    Module.Write Mass=true
    Module.Write Material Random=true
    Module.Write Position=true
    Module.Write Scale=false
    Module.Write SpriteRotation=true
    Module.Write SpriteSize=true
    Position=&SimulationPosition()
    Sprite Rotation=&NormalizedAngleToDegrees(NormalizedAngle=&UniformRangedFloat(Minimum=0, Maximum=360))
    Sprite Size=&Multiply_Vector2(
     A=<40 40>
     B=&Vector2DFromFloat(Value=&UniformRangedFloat(Minimum=1.1, Maximum=1.5))
    )
   )
   /NiagaraFN/Sprite/FN_SpriteAlignmentAndFacing(SpriteFacing:Vector3f=<0 0 -1>)
   &SystemLocation(MaskSpawnGroup=false, Offset=<0 0 2>)
  ]
  particleUpdateScript [
   &ParticleState()
   &ScaleSpriteSize(Scale Factor=&Vector2DFromFloat(Value=&FloatFromCurve(FloatCurve=DataInterface() {propertyValues {Curve {keys [
          {arriveTangent=1.4888, interpMode=RCIM_Cubic, leaveTangent=1.4888, tangentMode=RCTM_User, time=-0.0013, value=0.29}
          {arriveTangent=0.283, interpMode=RCIM_Cubic, leaveTangent=0.283, tangentMode=RCTM_User, time=0.9981, value=1.0141}
         ]}}})))
   &Color(
    Color=<rgba 1 100 1 1>
    Scale Color=&VectorFromCurve(VectorCurve=DataInterface() {propertyValues {
       XCurve {keys [<at -0.0031 48.8>,<at 1 1>]}
       YCurve {keys [<at -0.006 20.3>,<at 1 0.8>]}
       ZCurve {keys [<at -0.0011 1.0067>,<at 0.9989 0.28>]}
      }})
   ) {enabled=false}
  ]
  rendererClasses [&NiagaraSpriteRendererProperties]
  renderers [{rendererClass=&NiagaraSpriteRendererProperties}]
 }
}
summary {
 systemName=NS_Emote_TennisPaddle
 emitters [
  {bEnabled=false, emitterName=Smoke_RadialLocal, rendererClasses [&NiagaraMeshRendererProperties]}
  {bEnabled=false, emitterName=Smoke_Linear, rendererClasses [&NiagaraMeshRendererProperties]}
  {emitterName=Wind_Boof, rendererClasses [&NiagaraSpriteRendererProperties]}
  {bEnabled=false, emitterName=SpikeStreakMesh_NEW, rendererClasses [&NiagaraMeshRendererProperties]}
  {emitterName=SmokeRadial, rendererClasses [&NiagaraSpriteRendererProperties]}
  {emitterName=SmokeLinear, rendererClasses [&NiagaraSpriteRendererProperties]}
 ]
 userVariables [
  User.GlobalOffset {type=Vector3f, description="", defaultValue {struct=Vector3f, value=<0 0 0>}}
  User.RibbonIDOffset {type=NiagaraInt32}
 ]
}
